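# Exposed material function that builds the scorched-earth crater layer for landscape materials, outputting material attributes plus a separate Normal.
# A scorched-earth diffuse texture runs through CraterColoration, weighted by the Layer input, to fill only the attributes' BaseColor.
# A large normal on separately scaled UVs is angle-correct blended with a detail normal.
# Quality and feature-level switches drop to the large normal alone on low quality and mobile.
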MaterialFunction /BlastBerryMapContent/Game/Athena/Environments/Materials/MaterialFunctions/MF_Athena_Crater_01
graph {
 properties {bExposeToLibrary=true}
 nodes [
  n1 = FunctionInput(Preview=TextureCoordinate()) {bUsePreviewValueAsDefault=true, inputName=UVs, inputType=FunctionInput_Vector2}
  n2 = TextureSample(UVs=@n1) {texture=/Game/Packages/DS_Fortnight/T_Crater_Scorched_Earth2_D}
  n3 = FunctionInput(Preview=LandscapeLayerSample() {parameterName=Crater}) {inputName=Layer, inputType=FunctionInput_Scalar}
  FunctionOutput(None=MakeMaterialAttributes!(
    BaseColor=MaterialFunctionCall!(In, BaseColor=@n2.RGB, LayerWeight=@n3) {materialFunction=/Game/Athena/Environments/Landscape/Material/CraterColoration}.BaseColorOutput
    Metallic
    Specular
    Roughness
    Anisotropy
    EmissiveColor
    Opacity
    OpacityMask
    Normal
    Tangent
    WorldPositionOffset
    SubsurfaceColor
    ClearCoat
    ClearCoatRoughness
    AmbientOcclusion
    Refraction
    CustomizedUVs_0
    CustomizedUVs_1
    CustomizedUVs_2
    CustomizedUVs_3
    CustomizedUVs_4
    CustomizedUVs_5
    CustomizedUVs_6
    CustomizedUVs_7
    PixelDepthOffset
    ShadingModel
    Displacement
   ))
  n4 = TextureCoordinate()
  n5 = TextureSample(UVs=Multiply(
    A=FunctionInput(Preview=@n4) {bUsePreviewValueAsDefault=true, inputName=UV_2ndNormal, inputType=FunctionInput_Vector2}
    B=ScalarParameter() {defaultValue=0.015, parameterName=CraterNormaltiling}
   )) {samplerType=SAMPLERTYPE_Normal, texture=/BlastBerryMapContent/Game/Athena/Environments/Landscape/Textures/T_Crater_Scorched_Earth2_Large_N}
  n6 = MaterialFunctionCall(
   BaseNormal=@n5.RGB
   AdditionalNormal=TextureSample(UVs=@n1) {samplerType=SAMPLERTYPE_Normal, texture=/Game/Packages/DS_Fortnight/T_Crater_Scorched_Earth2_N}.RGB
  ) {materialFunction=&BlendAngleCorrectedNormals}
  n7 = QualitySwitch(Default=@n6.Result, Low=@n5.RGB, High=@n6.Result, Medium=@n6.Result)
  n8 = FeatureLevelSwitch(Default=@n7, ES2=@n5.RGB, ES3_1=@n5.RGB, SM4_REMOVED=@n7)
  FunctionOutput(None=@n8) {outputName=Normal}
  LandscapeLayerCoords() {mappingScale=6}
  LandscapeLayerCoords() {mappingScale=1}
  LinearInterpolate(B=@n8)
 ]
}
