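# Landscape master material for BlastBerry Athena terrain, blending painted grass, farm, mountain, arid, gravel, mud, crater and golf layers.
# Slope-driven height lerps mix grass with rock and sand, while MPC tints, a biome colormap and distance-faded macro normal add variation.
# Layer weights minus rock and exclusion masks drive Landscape Grass Output foliage.
# Quality, feature-level and static switches provide cheaper mobile and quick-color paths, alongside runtime virtual texture sampling.
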Material /BlastBerryMapContent/Game/Athena/Environments/Materials/M_Athena_Terrain_Master
graph {
 properties {
  bUseMaterialAttributes=true
  bUsedWithNanite=true
  pixelDepthOffsetMode=PDOM_Legacy
  refractionMethod=RM_IndexOfRefraction
  usedShadingModels=Default Lit
 }
 nodes [
  n1 = LandscapeLayerSample() {parameterName=Grass_Topology, previewWeight=1}
  n2 = LandscapeLayerSample() {parameterName=Grass_Only, previewWeight=1}
  n3 = Round(None=@n2)
  n4 = ScalarParameter() {defaultValue=-65, parameterName=ShiftGrassEdgeLOW}
  n5 = ScalarParameter() {defaultValue=50, parameterName=ShiftGrassEdgeHigh}
  n6 = MaterialFunctionCall() {materialFunction=/BlastBerryMapContent/Game/Athena/Environments/Landscape/MaterialFunctions/Standard/MF_Athena_Grass_01}
  n7 = MaterialFunctionCall!(UseRoughness, UVs=LandscapeLayerCoords() {mappingScale=5}) {materialFunction=/BlastBerryMapContent/Game/Athena/Environments/Landscape/MaterialFunctions/Arid/MF_Athena_Arid_Sand}
  n8 = Subtract(A=WorldPosition() {worldPositionShaderOffset=WPT_ExcludeAllShaderOffsets}.Z, B=Constant() {r=10}) {constB=-50}
  n9 = Constant() {r=10}
  n10 = OneMinus(None=Clamp(None=Divide(A=@n8, B=@n9) {constB=50}))
  n11 = MaterialFunctionCall!(
   Material=MaterialFunctionCall!(Material=@n7.Simple_Material, NormalFlatness=Constant() {r=0.75}, Normal) {materialFunction=/Engine/Functions/MaterialLayerFunctions/MatLayerBlend_NormalFlatten}.Blended Material
   NewBaseColor=Constant() {r=0.75}
   Mask=@n10
  ) {materialFunction=/Engine/Functions/MaterialLayerFunctions/MatLayerBlend_MultiplyBaseColor}
  n12 = MaterialFunctionCall(
   Transition Phase=LandscapeLayerSample() {parameterName=Sand}
   Height Texture=OneMinus(None=@n6.Height)
   Contrast=ScalarParameter() {defaultValue=0.5, parameterName=SandBlendContrast}
  ) {materialFunction=&HeightLerp}
  n13 = BlendMaterialAttributes(A=@n6.Result, B=@n11.Blended Material, Alpha=@n12.Alpha)
  n14 = MaterialFunctionCall!(height_Offset, height_GradScale, Coords=WorldPosition().XYZ) {materialFunction=/Game/Athena/Environments/Landscape/MaterialFunctions/MF_Athena_WorldHeightGrad}
  n15 = Constant()
  n16 = MaterialFunctionCall() {materialFunction=/BlastBerryMapContent/Game/Athena/Environments/Landscape/MaterialFunctions/Standard/MF_Athena_Rock_01}
  n17 = MaterialFunctionCall!(
   In=@n16.Result
   Color=LinearInterpolate(
    A=CollectionParameter() {collection=/BlastBerryMapContent/Game/Athena/Environments/Landscape/MPC/MPC_Landscape, parameterName=Rock_Bottom_Brightness}
    B=CollectionParameter() {collection=&MPC_Landscape, parameterName=Rock_Top_Brightness}
    Alpha=MaterialFunctionCall(Transition Phase=@n14.Result, Height Texture=@n15, Contrast=@n15) {materialFunction=&HeightLerp}.Alpha
   )
  ) {materialFunction=/Game/Athena/Environments/Landscape/MaterialFunctions/MF_Athena_Material_BaseMultiply}
  n18 = Multiply(
   A=Multiply(
    A=MaterialFunctionCall!(Mat attributes=@n13) {materialFunction=/Engine/Functions/Engine_MaterialFunctions02/MatLayerBlend_BreakOpacity}.Result
    B=MaterialFunctionCall(Mat attributes=@n17.Result) {materialFunction=&MatLayerBlend_BreakOpacity}.Result
   )
   B=ScalarParameter() {defaultValue=0.08, group=Grass, parameterName=Transition_Phase}
  )
  n19 = Multiply(A=ComponentMask(None=VertexNormalWS()) {b=true}, B=ScalarParameter() {defaultValue=2, group=Rock, parameterName=RockMultiply_2})
  n20 = ScalarParameter() {defaultValue=5, group=Grass, parameterName=Contrast}
  n21 = MaterialFunctionCall(Transition Phase=@n18, Height Texture=@n19, Contrast=@n20) {materialFunction=&HeightLerp}
  n22 = Multiply(A=@n21.Lerp Alpha No Contrast) {desc=Rock Alpha for Grass system}
  n23 = Add(A=@n22, B=ScalarParameter() {defaultValue=-0.24, parameterName=ShiftGrassEdge_2})
  n24 = OneMinus(None=Clamp(None=LinearInterpolate(A=@n4, B=@n5, Alpha=@n23)))
  n25 = LandscapeLayerSample() {parameterName=Gravel}
  n26 = LandscapeLayerSample() {parameterName=Mud}
  n27 = LandscapeLayerSample() {parameterName=Road, previewWeight=1}
  n28 = Add(A=Add(A=Add(A=@n25, B=@n24), B=@n26), B=@n27)
  n29 = LandscapeLayerSample() {parameterName=NoFoliage}
  n30 = Add(A=Ceil(None=LandscapeLayerSample() {parameterName=Golf_Fairway}), B=Ceil(None=LandscapeLayerSample() {parameterName=Golf_Green}))
  n31 = Add(A=Add(A=Add(A=@n28, B=@n29), B=LandscapeLayerSample() {parameterName=Forest}), B=Round(None=@n30))
  n32 = LandscapeLayerSample() {parameterName=Sand}
  n33 = Subtract(A=Subtract(A=Add(A=@n1, B=@n3), B=Round(None=@n24)), B=Add(A=@n31, B=Round(None=@n32)))
  n34 = LandscapeLayerSample() {parameterName=Farm_Grass}
  n35 = Subtract(A=Add(A=LandscapeLayerSample() {parameterName=Mountain_Grass}, B=@n34), B=@n31)
  n36 = Add(A=Add(A=@n33, B=Subtract(A=@n34, B=@n31)), B=@n35)
  n37 = Round(None=LandscapeLayerSample() {parameterName=Arid})
  Add(A=@n36, B=@n37)
  ComponentMask(None=CollectionParameter() {collection=&MPC_Landscape, parameterName=Arid_Grass_Tint}) {b=true, g=true, r=true}
  Constant3Vector() {constant=<rgba 1 0 0.7404 1>}
  Constant()
  Constant() {r=20000}
  Constant()
  n38 = ComponentMask(None=WorldPosition().XYZ) {g=true, r=true}
  n39 = ScalarParameter() {defaultValue=306000, parameterName=LandscapeSize, sortPriority=-153000}
  n40 = ScalarParameter() {defaultValue=-153000, parameterName=LandscapeMin, sortPriority=-153000}
  n41 = TextureSampleParameter2D(UVs=Subtract(A=Divide(A=@n38, B=@n39), B=Divide(A=@n40, B=@n39))) {
   parameterName=BiomeColormap
   samplerType=SAMPLERTYPE_LinearColor
   texture=/BlastBerryMapContent/Game/Athena/Environments/Landscape/Textures/T_Athena_Terrain_CombinedColors_01
  }
  n42 = LinearInterpolate(A=@n6.BaseColor_Plain, B=@n7.BaseColor_Plain, Alpha=@n12.Alpha)
  MaterialExpressionReroute_24 = Reroute(None=@n42)
  MaterialExpressionReroute_8 = Reroute(None=LinearInterpolate(A=Multiply(A=@n16.BaseColor_Plain), B=@n42, Alpha=@n21.Alpha))
  n43 = ScalarParameter() {defaultValue=10, group=Arid_Rock, parameterName=Arid_Rock_ProjContrast}
  n44 = MaterialFunctionCall!(ProjContrast=@n43) {materialFunction=/BlastBerryMapContent/Game/Athena/Environments/Landscape/MaterialFunctions/Arid/MF_Athena_Arid_Rock}
  n45 = MaterialFunctionCall!(ProjContrast=@n43) {materialFunction=/BlastBerryMapContent/Game/Athena/Environments/Landscape/MaterialFunctions/Arid/MF_Athena_Arid_Rock_02}
  n46 = LandscapeLayerSample() {parameterName=Arid_RockDetail, previewWeight=1}
  n47 = MaterialFunctionCall() {materialFunction=/BlastBerryMapContent/Game/Athena/Environments/Landscape/MaterialFunctions/MF_Athena_SedimentGradient}
  n48 = Multiply(A=LinearInterpolate(A=@n44.BaseColor_Low, B=@n45.BaseColor_Low, Alpha=@n46), B=@n47.Result)
  n49 = Multiply(
   A=QualitySwitch(Default=@n48, Low=Multiply(A=@n44.BaseColor_Low, B=@n47.Result), Medium=@n48)
   B=ScalarParameter() {defaultValue=1.5, parameterName=LowMed_RockBrightness}
  )
  n50 = VectorParameter() {
   desc=Replacement adjustment for mobile without Sediment/Spec/Normal on PC
   parameterName=SedimentMobileReplace
   defaultValue=<rgba 1 0.7993 0.7031 0>
  }
  n51 = Multiply(A=@n44.BaseColor_Low, B=@n50.RGB)
  n52 = Abs(None=ScalarParameter() {defaultValue=1500, group=MasterControl, parameterName=Texturesize_Z_Sand})
  n53 = ComponentMask(None=Divide(A=WorldPosition() {worldPositionShaderOffset=WPT_ExcludeAllShaderOffsets}.XYZ, B=Multiply(A=@n52) {constB=-1})) {g=true, r=true}
  n54 = LandscapeLayerCoords() {mappingScale=5}
  n55 = QualitySwitch(Default=@n53, Low=@n54, High=@n53, Medium=@n53)
  n56 = MaterialFunctionCall(UseRoughness=StaticBool(), UVs=FeatureLevelSwitch(Default=@n55, ES2=@n54, ES3_1=@n54, SM4_REMOVED=@n55, SM5=@n53, SM6=@n53)) {materialFunction=&MF_Athena_Arid_Sand}
  n57 = ScalarParameter() {defaultValue=-1.5, group=Arid_SandRockBlend, parameterName=Arid_SandRock_LOW}
  n58 = ScalarParameter() {defaultValue=1.5, group=Arid_SandRockBlend, parameterName=Arid_SandRock_High}
  n59 = ComponentMask(None=VertexNormalWS()) {b=true}
  n60 = Clamp(None=Add(
    A=LinearInterpolate(A=@n57, B=@n58, Alpha=@n59)
    B=ScalarParameter() {defaultValue=-0.5429, group=Arid_SandRockBlend, parameterName=Arid_SandRock_Offset}
   ))
  n61 = OneMinus(None=@n44.Height)
  n62 = ScalarParameter() {defaultValue=3, parameterName=Desert_SandRock_Contrast}
  n63 = MaterialFunctionCall(Transition Phase=@n60, Height Texture=@n61, Contrast=@n62) {desc=Base Color High end, materialFunction=&HeightLerp}
  MaterialExpressionStaticSwitchParameter_2 = StaticSwitchParameter(
   True=LinearInterpolate(A=FeatureLevelSwitch(Default=@n49, ES2=@n51, ES3_1=@n51, SM4_REMOVED=@n49, SM5=@n49, SM6=@n49), B=@n56.BaseColor, Alpha=@n63.Alpha) {desc=Base Color only}
   False=@n56.BaseColor
  ) {defaultValue=true, parameterName=UseAridRockBlend}
  n64 = LandscapeLayerBlend!(
   Layer Grass_Only=@MaterialExpressionReroute_24
   Layer Grass_Topology=@MaterialExpressionReroute_8
   Layer Arid=@MaterialExpressionStaticSwitchParameter_2
   Layer Farm_Grass=@MaterialExpressionReroute_8
   Layer Mountain_Grass=@MaterialExpressionReroute_8
  ) {layers [
    {
     blendType=LB_WeightBlend
     constHeightInput=0
     layerName=Grass_Only
     previewWeight=1
     constLayerInput=<0 0 0>
     heightInput {expression=null, inputName=None, mask=0, maskA=0, maskB=0, maskG=0, maskR=0, outputIndex=0}
     layerInput {
      expression=&self:MaterialExpressionReroute_24
      inputName=None
      mask=0
      maskA=0
      maskB=0
      maskG=0
      maskR=0
      outputIndex=0
     }
    }
    {
     blendType=LB_WeightBlend
     constHeightInput=0
     layerName=Grass_Topology
     previewWeight=1
     constLayerInput=<0 0 0>
     heightInput {expression=null, inputName=None, mask=0, maskA=0, maskB=0, maskG=0, maskR=0, outputIndex=0}
     layerInput {
      expression=&self:MaterialExpressionReroute_8
      inputName=None
      mask=0
      maskA=0
      maskB=0
      maskG=0
      maskR=0
      outputIndex=0
     }
    }
    {
     blendType=LB_WeightBlend
     constHeightInput=0
     layerName=Arid
     previewWeight=1
     constLayerInput=<0 0 0>
     heightInput {expression=null, inputName=None, mask=0, maskA=0, maskB=0, maskG=0, maskR=0, outputIndex=0}
     layerInput {
      expression=&self:MaterialExpressionStaticSwitchParameter_2
      inputName=None
      mask=0
      maskA=0
      maskB=0
      maskG=0
      maskR=0
      outputIndex=0
     }
    }
    {
     blendType=LB_WeightBlend
     constHeightInput=0
     layerName=Farm_Grass
     previewWeight=1
     constLayerInput=<0 0 0>
     heightInput {expression=null, inputName=None, mask=0, maskA=0, maskB=0, maskG=0, maskR=0, outputIndex=0}
     layerInput {
      expression=&self:MaterialExpressionReroute_8
      inputName=None
      mask=0
      maskA=0
      maskB=0
      maskG=0
      maskR=0
      outputIndex=0
     }
    }
    {
     blendType=LB_WeightBlend
     constHeightInput=0
     layerName=Mountain_Grass
     previewWeight=1
     constLayerInput=<0 0 0>
     heightInput {expression=null, inputName=None, mask=0, maskA=0, maskB=0, maskG=0, maskR=0, outputIndex=0}
     layerInput {
      expression=&self:MaterialExpressionReroute_8
      inputName=None
      mask=0
      maskA=0
      maskB=0
      maskG=0
      maskR=0
      outputIndex=0
     }
    }
   ]}
  n65 = ComponentMask(None=CollectionParameter() {collection=&MPC_Landscape, parameterName=Arid_Grass_Tint}) {b=true, g=true, r=true}
  n66 = Multiply(A=@n65, B=@MaterialExpressionReroute_24)
  n67 = LandscapeLayerSample() {parameterName=Arid_Grass_Light}
  MaterialExpressionLinearInterpolate_52 = LinearInterpolate(A=ScalarParameter() {defaultValue=0.8, group=Grass, parameterName=Grass_Tint}, Alpha=LandscapeLayerSample() {parameterName=NoTopo})
  MaterialExpressionMaterialFunctionCall_77 = MaterialFunctionCall(In=@n13, Color=@MaterialExpressionLinearInterpolate_52) {materialFunction=&MF_Athena_Material_BaseMultiply}
  n68 = MaterialFunctionCall(
   Transition Phase=@n67
   Height Texture=MaterialFunctionCall(Mat attributes=@MaterialExpressionMaterialFunctionCall_77.Result) {materialFunction=&MatLayerBlend_BreakOpacity}.Result
   Contrast=ScalarParameter() {parameterName=Arid_Grass_Contrast}
  ) {materialFunction=&HeightLerp}
  n69 = ScalarParameter() {defaultValue=1.5, parameterName=AridGrassMob_MP}
  n70 = Clamp(None=Multiply(A=@n67, B=@n69))
  n71 = QualitySwitch(Default=@n68.Alpha, Low=@n70, Medium=@n68.Alpha) {desc="Use Height lerp for Medium, but simple lerp for Mobile/Low"}
  n72 = FeatureLevelSwitch(Default=@n71, ES2=@n70, ES3_1=@n70, SM4_REMOVED=@n71, SM5=@n71, SM6=@n71)
  n73 = MakeMaterialAttributes!(
   BaseColor=LandscapeLayerSwitch!(LayerUsed=LinearInterpolate(A=@n64, B=@n66, Alpha=@n72), LayerNotUsed=@n64) {parameterName=Arid_Grass_Light}
   Metallic
   Specular
   Roughness
   Anisotropy
   EmissiveColor
   Opacity
   OpacityMask
   Normal
   Tangent
   WorldPositionOffset
   SubsurfaceColor
   ClearCoat
   ClearCoatRoughness
   AmbientOcclusion
   Refraction
   CustomizedUVs_0
   CustomizedUVs_1
   CustomizedUVs_2
   CustomizedUVs_3
   CustomizedUVs_4
   CustomizedUVs_5
   CustomizedUVs_6
   CustomizedUVs_7
   PixelDepthOffset
   ShadingModel
   Displacement
  )
  n74 = LandscapeLayerSample() {parameterName=Golf_Fairway, previewWeight=1}
  n75 = MaterialFunctionCall!(In_LowMobile_Material=@n73, Golf_Fairway_Mask=@n74, In) {
   desc=Simple version of fairway for LOW/Medium
   materialFunction=/BlastBerryMapContent/Game/Athena/Environments/Materials/MaterialFunctions/MF_LawnGrassColoration
  }
  n76 = LandscapeLayerSwitch(LayerUsed=@n75.LowMobile_Fairway, LayerNotUsed=@n73) {parameterName=Golf_Fairway}
  n77 = MaterialFunctionCall(
   In=@n76
   Color=Multiply(
    A=@n41.RGB
    B=CollectionParameter() {collection=&MPC_Landscape, parameterName=ColormapDivisor}
   )
  ) {materialFunction=&MF_Athena_Material_BaseMultiply}
  n78 = CollectionParameter() {collection=&MPC_Landscape, parameterName=DisplayLowDetailShader}
  MaterialExpressionMaterialFunctionCall_74 = MaterialFunctionCall(Mat attributes=@n13) {materialFunction=&MatLayerBlend_BreakOpacity}
  n79 = ComponentMask(None=CollectionParameter() {collection=&MPC_Landscape, parameterName=RockTint}) {b=true, g=true, r=true}
  n80 = MaterialFunctionCall!(Base Material=@n17.Result, Top Material=@n13, Alpha=@n21.Alpha) {materialFunction=/Engine/Functions/MaterialLayerFunctions/MatLayerBlend_Standard}
  MaterialExpressionMaterialFunctionCall_72 = MaterialFunctionCall(In=@n80.Blended Material, Color=LinearInterpolate(A=@n79, B=@MaterialExpressionLinearInterpolate_52, Alpha=@n21.Alpha)) {materialFunction=&MF_Athena_Material_BaseMultiply}
  n81 = MaterialFunctionCall(Base Material=@n44.Mat, Top Material=@n45.Mat, Alpha=@n46) {materialFunction=&MatLayerBlend_Standard}
  n82 = Multiply(A=@n47.Result, B=ScalarParameter() {defaultValue=1, group=Arid_Rock, parameterName=Arid_ClifDesaturate_All})
  n83 = MaterialFunctionCall() {materialFunction=/BlastBerryMapContent/Game/Athena/Environments/Landscape/MaterialFunctions/MF_Athena_ZSandColor}
  n84 = MaterialFunctionCall(In=@n56.Simple_Material, Color=@n83.Result) {materialFunction=&MF_Athena_Material_BaseMultiply}
  MaterialExpressionStaticSwitchParameter_4 = StaticSwitchParameter(
   True=MaterialFunctionCall(
    Base Material=MaterialFunctionCall(In=@n81.Blended Material, Color=@n82) {materialFunction=&MF_Athena_Material_BaseMultiply}.Result
    Top Material=@n84.Result
    Alpha=@n63.Alpha
   ) {desc=Full Mat, materialFunction=&MatLayerBlend_Standard}.Blended Material
   False=@n84.Result
  ) {defaultValue=true, parameterName=UseAridRockBlend}
  n85 = MaterialFunctionCall(
   Transition Phase=Add(
    A=CollectionParameter() {collection=&MPC_Landscape, parameterName=ColormapTransitionBias}
    B=@n18
   )
   Height Texture=@n19
   Contrast=@n20
  ) {
   desc=Height lerp for ColorMap Texture - has ColormapTransitionBias
   materialFunction=&HeightLerp
  }
  MaterialExpressionMaterialFunctionCall_18 = MaterialFunctionCall!(Rock_Diffuse, Alpha, ColorBlend_Alpha=@n21.Alpha, ColorBlend_Alpha_Bias=@n85.Alpha, In, Grass_Diffuse) {materialFunction=/BlastBerryMapContent/Game/Athena/Environments/Materials/MaterialFunctions/MF_FarmGrass_Colors}
  MaterialExpressionMaterialFunctionCall_67 = MaterialFunctionCall(In=@n80.Blended Material, Color=@MaterialExpressionMaterialFunctionCall_18.BlendedColors) {materialFunction=&MF_Athena_Material_BaseMultiply}
  MaterialExpressionMaterialFunctionCall_23 = MaterialFunctionCall!(Rock_Diffuse, ColorBlend_Alpha=@n21.Alpha, ColorBlend_Alpha_Bias=@n85.Alpha, Rock_Diffuse1, Grass_Diffuse) {materialFunction=/BlastBerryMapContent/Game/Athena/Environments/Landscape/MaterialFunctions/MF_MountainGrass_Colors}
  MaterialExpressionMaterialFunctionCall_68 = MaterialFunctionCall(In=@n80.Blended Material, Color=@MaterialExpressionMaterialFunctionCall_23.Blended_Colors) {materialFunction=&MF_Athena_Material_BaseMultiply}
  n86 = LandscapeLayerBlend!(
   Layer Grass_Only=@MaterialExpressionMaterialFunctionCall_77.Result
   Height Grass_Only=@MaterialExpressionMaterialFunctionCall_74.Result
   Layer Grass_Topology=@MaterialExpressionMaterialFunctionCall_72.Result
   Layer Arid=@MaterialExpressionStaticSwitchParameter_4
   Layer Farm_Grass=@MaterialExpressionMaterialFunctionCall_67.Result
   Layer Mountain_Grass=@MaterialExpressionMaterialFunctionCall_68.Result
  ) {layers [
    {
     blendType=LB_HeightBlend
     constHeightInput=0
     layerName=Grass_Only
     previewWeight=1
     constLayerInput=<0 0 0>
     heightInput {
      expression=&self:MaterialExpressionMaterialFunctionCall_74
      inputName=None
      mask=0
      maskA=0
      maskB=0
      maskG=0
      maskR=0
      outputIndex=0
     }
     layerInput {
      expression=&self:MaterialExpressionMaterialFunctionCall_77
      inputName=None
      mask=0
      maskA=0
      maskB=0
      maskG=0
      maskR=0
      outputIndex=0
     }
    }
    {
     blendType=LB_WeightBlend
     constHeightInput=0
     layerName=Grass_Topology
     previewWeight=1
     constLayerInput=<0 0 0>
     heightInput {expression=null, inputName=None, mask=0, maskA=0, maskB=0, maskG=0, maskR=0, outputIndex=0}
     layerInput {
      expression=&self:MaterialExpressionMaterialFunctionCall_72
      inputName=None
      mask=0
      maskA=0
      maskB=0
      maskG=0
      maskR=0
      outputIndex=0
     }
    }
    {
     blendType=LB_WeightBlend
     constHeightInput=0
     layerName=Arid
     previewWeight=1
     constLayerInput=<0 0 0>
     heightInput {expression=null, inputName=None, mask=0, maskA=0, maskB=0, maskG=0, maskR=0, outputIndex=0}
     layerInput {
      expression=&self:MaterialExpressionStaticSwitchParameter_4
      inputName=None
      mask=0
      maskA=0
      maskB=0
      maskG=0
      maskR=0
      outputIndex=0
     }
    }
    {
     blendType=LB_WeightBlend
     constHeightInput=0
     layerName=Farm_Grass
     previewWeight=1
     constLayerInput=<0 0 0>
     heightInput {expression=null, inputName=None, mask=0, maskA=0, maskB=0, maskG=0, maskR=0, outputIndex=0}
     layerInput {
      expression=&self:MaterialExpressionMaterialFunctionCall_67
      inputName=None
      mask=0
      maskA=0
      maskB=0
      maskG=0
      maskR=0
      outputIndex=0
     }
    }
    {
     blendType=LB_WeightBlend
     constHeightInput=0
     layerName=Mountain_Grass
     previewWeight=1
     constLayerInput=<0 0 0>
     heightInput {expression=null, inputName=None, mask=0, maskA=0, maskB=0, maskG=0, maskR=0, outputIndex=0}
     layerInput {
      expression=&self:MaterialExpressionMaterialFunctionCall_68
      inputName=None
      mask=0
      maskA=0
      maskB=0
      maskG=0
      maskR=0
      outputIndex=0
     }
    }
   ]}
  n87 = MaterialFunctionCall(
   Base Material=@n86
   Top Material=MaterialFunctionCall(In=@MaterialExpressionMaterialFunctionCall_77.Result, Color=@n65) {materialFunction=&MF_Athena_Material_BaseMultiply}.Result
   Alpha=@n68.Alpha
  ) {materialFunction=&MatLayerBlend_Standard}
  n88 = MaterialFunctionCall(Golf_Fairway_Mask=@n74, In=@n87.Blended Material) {materialFunction=&MF_LawnGrassColoration}
  n89 = LandscapeLayerSwitch(LayerUsed=@n88.Result, LayerNotUsed=@n87.Blended Material) {parameterName=Golf_Fairway}
  n90 = StaticSwitchParameter(
   True=MaterialFunctionCall!(
    Material=@n89
    NewBaseColor=MaterialFunctionCall!(Material=@n77.Result) {materialFunction=/Engine/Functions/MaterialLayerFunctions/MatLayerBlend_BreakBaseColor}.BaseColor
    Mask=Saturate(None=@n78)
   ) {materialFunction=/Engine/Functions/MaterialLayerFunctions/MatLayerBlend_OverrideBaseColor}.Blended Material
   False=@n89
  ) {group=Utilities, parameterName=EnableQuickColorRender}
  FeatureLevelSwitch(Default=@n90, ES2=@n77.Result, ES3_1=@n77.Result, SM4_REMOVED=@n90, SM5=@n90, SM6=@n90)
  Floor(None=@n2)
  MaterialExpressionReroute_32 = Reroute(None=@n33)
  MaterialExpressionReroute_35 = Reroute(None=@n35)
  n91 = LandscapeLayerSample() {parameterName=NoFoliage, previewWeight=1}
  MaterialExpressionSubtract_4 = Subtract(A=LandscapeLayerSample() {parameterName=CraterOvergrowth}, B=@n91)
  n92 = Add(A=Add(A=@n26, B=@n27), B=@n30)
  MaterialExpressionSubtract_5 = Subtract(A=Subtract(A=LandscapeLayerSample() {parameterName=Forest}, B=Add(A=@n92, B=@n29)), B=Round(None=@n24))
  n93 = LandscapeLayerSample() {parameterName=Gravel}
  MaterialExpressionLandscapeLayerSample_40 = LandscapeLayerSample() {parameterName=Crater}
  n94 = LandscapeLayerSample() {parameterName=NoFoliage}
  MaterialExpressionSubtract_3 = Subtract(A=Subtract(A=Subtract(A=@n93, B=@n24), B=Multiply(B=@MaterialExpressionLandscapeLayerSample_40) {constA=10}), B=@n94)
  MaterialExpressionSubtract_70 = Subtract(A=LandscapeLayerSample() {parameterName=Flowers, previewWeight=1}, B=@n91)
  MaterialExpressionClamp_10 = Clamp(None=@n36)
  n95 = Subtract(A=@n37, B=@n24) {desc=Subtract rock}
  MaterialExpressionSubtract_1 = Subtract(A=@n95, B=@n31)
  MaterialExpressionSubtract_6 = Subtract(A=Max(A=@MaterialExpressionReroute_32, B=@MaterialExpressionReroute_35), B=LandscapeLayerSample() {parameterName=NoTallPlant})
  LandscapeGrassOutput!(
   Grass=@MaterialExpressionReroute_32
   Mountain=@MaterialExpressionReroute_35
   Farm=@MaterialExpressionSubtract_4
   Forest=@MaterialExpressionSubtract_5
   Gravel=@MaterialExpressionSubtract_3
   Flowers=@MaterialExpressionSubtract_70
   Bushes=@MaterialExpressionClamp_10
   Crater=@MaterialExpressionLandscapeLayerSample_40
   Arid_Sand=@MaterialExpressionSubtract_1
   TallPlant=@MaterialExpressionSubtract_6
  ) {grassTypes [
    Grass {
     grassType=/BlastBerryMapContent/Game/Athena/Environments/Landscape/Foliage/Athena_Foliage_Grass
     input {
      expression=&self:MaterialExpressionReroute_32
      inputName=None
      mask=0
      maskA=0
      maskB=0
      maskG=0
      maskR=0
      outputIndex=0
     }
    }
    Mountain {
     grassType=/BlastBerryMapContent/Game/Athena/Environments/Landscape/Foliage/Athena_Foliage_Mountain_Grass1
     input {
      expression=&self:MaterialExpressionReroute_35
      inputName=None
      mask=0
      maskA=0
      maskB=0
      maskG=0
      maskR=0
      outputIndex=0
     }
    }
    Farm {
     grassType=/Game/Athena/Environments/Landscape/Foliage/Athena_Foliage_CraterOvergrowth
     input {
      expression=&self:MaterialExpressionSubtract_4
      inputName=None
      mask=0
      maskA=0
      maskB=0
      maskG=0
      maskR=0
      outputIndex=0
     }
    }
    Forest {
     grassType=/BlastBerryMapContent/Game/Athena/Environments/Landscape/Foliage/Athena_Foliage_Forest
     input {
      expression=&self:MaterialExpressionSubtract_5
      inputName=None
      mask=0
      maskA=0
      maskB=0
      maskG=0
      maskR=0
      outputIndex=0
     }
    }
    Gravel {
     grassType=/Game/Athena/Environments/Landscape/Foliage/Athena_Foliage_Gravel
     input {
      expression=&self:MaterialExpressionSubtract_3
      inputName=None
      mask=0
      maskA=0
      maskB=0
      maskG=0
      maskR=0
      outputIndex=0
     }
    }
    Flowers {
     grassType=/BlastBerryMapContent/Game/Athena/Environments/Landscape/Foliage/Rufus_Foliage_Flowers
     input {
      expression=&self:MaterialExpressionSubtract_70
      inputName=None
      mask=0
      maskA=0
      maskB=0
      maskG=0
      maskR=0
      outputIndex=0
     }
    }
    Bushes {
     grassType=/BlastBerryMapContent/Game/Athena/Environments/Landscape/Foliage/Athena_Foliage_Bushes
     input {
      expression=&self:MaterialExpressionClamp_10
      inputName=None
      mask=0
      maskA=0
      maskB=0
      maskG=0
      maskR=0
      outputIndex=0
     }
    }
    Crater {
     grassType=/BlastBerryMapContent/Game/Athena/Environments/Landscape/Foliage/Athena_Foliage_Crater_Grass
     input {
      expression=&self:MaterialExpressionLandscapeLayerSample_40
      inputName=None
      mask=0
      maskA=0
      maskB=0
      maskG=0
      maskR=0
      outputIndex=0
     }
    }
    Arid_Sand {
     grassType=/BlastBerryMapContent/Game/Athena/Environments/Landscape/Foliage/Athena_Foliage_Arid_Grass
     input {
      expression=&self:MaterialExpressionSubtract_1
      inputName=None
      mask=0
      maskA=0
      maskB=0
      maskG=0
      maskR=0
      outputIndex=0
     }
    }
    TallPlant {
     grassType=/BlastBerryMapContent/Game/Athena/Environments/Landscape/Foliage/Rufus_TallPlant
     input {
      expression=&self:MaterialExpressionSubtract_6
      inputName=None
      mask=0
      maskA=0
      maskB=0
      maskG=0
      maskR=0
      outputIndex=0
     }
    }
   ]}
  LandscapeLayerCoords() {mappingScale=4}
  LandscapeLayerSample() {parameterName=Arid_Grass_Light}
  n96 = MaterialFunctionCall!(In=LandscapeLayerCoords() {mappingScale=4}) {materialFunction=/BlastBerryMapContent/Game/Athena/Environments/Landscape/MaterialFunctions/Standard/MF_Athena_Gravel_01}
  n97 = VectorParameter() {parameterName=Gravel_CraterColor, defaultValue=<rgba 0.8542 0.6349 0.4402 1>}
  n98 = Clamp(None=LinearInterpolate(
    A=ScalarParameter() {defaultValue=-0.2522, parameterName=MudCraterTransition_High}
    B=ScalarParameter() {defaultValue=10.999, parameterName=MudCraterTransition_LOW}
    Alpha=LandscapeLayerSample() {parameterName=Crater}
   ))
  n99 = LandscapeLayerSwitch(
   LayerUsed=MaterialFunctionCall(In=@n96.Result, Color=LinearInterpolate(B=@n97.RGB, Alpha=@n98) {constA=1}) {materialFunction=&MF_Athena_Material_BaseMultiply}.Result
   LayerNotUsed=@n96.Result
  ) {parameterName=Crater}
  n100 = CollectionParameter() {collection=&MPC_Landscape, parameterName=Arid_Gravel_Tint}
  n101 = MaterialFunctionCall(
   Material=@n99
   NewBaseColor=Multiply(A=@n96.Plain_BaseColor, B=ComponentMask(None=@n100) {b=true, g=true, r=true})
   Mask=LandscapeLayerSample() {parameterName=Arid}
  ) {desc=Plug in Arid, materialFunction=&MatLayerBlend_OverrideBaseColor}
  n102 = LandscapeLayerSwitch(LayerUsed=@n101.Blended Material, LayerNotUsed=@n99) {parameterName=Arid}
  n103 = MaterialFunctionCall(Material=@n102) {materialFunction=&MatLayerBlend_BreakBaseColor}
  n104 = LandscapeLayerSample() {parameterName=Golf_Fairway}
  LandscapeLayerSwitch(
   LayerUsed=MaterialFunctionCall(Material=@n102, NewBaseColor=LinearInterpolate(A=@n103.BaseColor, B=@n56.BaseColor, Alpha=@n104), Mask=@n104) {desc=Plug in Arid, materialFunction=&MatLayerBlend_OverrideBaseColor}.Blended Material
   LayerNotUsed=@n102
  ) {parameterName=Golf_Fairway}
  n105 = MaterialFunctionCall!(
   UVs=LandscapeLayerCoords() {mappingScale=6}
   UV_2ndNormal=LandscapeLayerCoords() {mappingScale=1}
   Layer=LandscapeLayerSample() {parameterName=Crater}
  ) {
   desc=Normal isnt part of main mat for now.
   materialFunction=/BlastBerryMapContent/Game/Athena/Environments/Materials/MaterialFunctions/MF_Athena_Crater_01
  }
  n106 = MaterialFunctionCall!(Material=@n87.Blended Material) {materialFunction=/Engine/Functions/MaterialLayerFunctions/MatLayerBlend_BreakNormal}
  n107 = ScalarParameter() {defaultValue=1, parameterName=Green_NormalFlatten}
  n108 = LandscapeLayerSwitch(
   LayerUsed=LinearInterpolate(
    A=@n106.Normal
    B=MaterialFunctionCall(Normal=@n106.Normal, Flatness=@n107) {materialFunction=&FlattenNormal}.Result
    Alpha=@n88.Green_Mask
   )
   LayerNotUsed=@n106.Normal
  ) {parameterName=Golf_Fairway}
  n109 = Constant3Vector() {constant=<rgba 0 0 1 0>}
  n110 = Add(A=LandscapeLayerSample() {parameterName=Road, previewWeight=1}, B=LandscapeLayerSample() {parameterName=Mud, previewWeight=1})
  n111 = LandscapeLayerSample() {parameterName=Gravel, previewWeight=1}
  n112 = LinearInterpolate(
   A=LandscapeLayerSwitch(LayerUsed=LinearInterpolate(A=@n108, B=@n105.Normal, Alpha=@n98), LayerNotUsed=@n108) {parameterName=Crater}
   B=@n109
   Alpha=Clamp(None=Add(A=@n110, B=@n111))
  )
  LinearInterpolate(A=@n112, B=@n109, Alpha=@n111)
  MaterialExpressionLinearInterpolate_56 = LinearInterpolate(A=@n79, B=@MaterialExpressionLinearInterpolate_52, Alpha=@n85.Alpha)
  n113 = Multiply(
   A=@n83.Result
   B=ComponentMask(None=CollectionParameter() {collection=&MPC_Landscape, parameterName=Arid_Sand_Tint}) {b=true, g=true, r=true}
  )
  MaterialExpressionStaticSwitchParameter_3 = StaticSwitchParameter(True=LinearInterpolate(A=Constant3Vector() {constant=<rgba 1 0.7455 0.5781 0>}, B=@n113, Alpha=@n63.Alpha) {desc=Colors Only}, False=@n113) {defaultValue=true, parameterName=UseAridRockBlend}
  n114 = LandscapeLayerBlend!(
   Layer Grass_Only=@MaterialExpressionLinearInterpolate_52
   Layer Grass_Topology=@MaterialExpressionLinearInterpolate_56
   Layer Arid=@MaterialExpressionStaticSwitchParameter_3
   Layer Farm_Grass=@MaterialExpressionMaterialFunctionCall_18.BlendedColors_Bias
   Layer Mountain_Grass=@MaterialExpressionMaterialFunctionCall_23.Blended_Colors_Bias
  ) {layers [
    {
     blendType=LB_WeightBlend
     constHeightInput=0
     layerName=Grass_Only
     previewWeight=1
     constLayerInput=<0 0 0>
     heightInput {expression=null, inputName=None, mask=0, maskA=0, maskB=0, maskG=0, maskR=0, outputIndex=0}
     layerInput {
      expression=&self:MaterialExpressionLinearInterpolate_52
      inputName=None
      mask=0
      maskA=0
      maskB=0
      maskG=0
      maskR=0
      outputIndex=0
     }
    }
    {
     blendType=LB_WeightBlend
     constHeightInput=0
     layerName=Grass_Topology
     previewWeight=1
     constLayerInput=<0 0 0>
     heightInput {expression=null, inputName=None, mask=0, maskA=0, maskB=0, maskG=0, maskR=0, outputIndex=0}
     layerInput {
      expression=&self:MaterialExpressionLinearInterpolate_56
      inputName=None
      mask=0
      maskA=0
      maskB=0
      maskG=0
      maskR=0
      outputIndex=0
     }
    }
    {
     blendType=LB_WeightBlend
     constHeightInput=0
     layerName=Arid
     previewWeight=1
     constLayerInput=<0 0 0>
     heightInput {expression=null, inputName=None, mask=0, maskA=0, maskB=0, maskG=0, maskR=0, outputIndex=0}
     layerInput {
      expression=&self:MaterialExpressionStaticSwitchParameter_3
      inputName=None
      mask=0
      maskA=0
      maskB=0
      maskG=0
      maskR=0
      outputIndex=0
     }
    }
    {
     blendType=LB_WeightBlend
     constHeightInput=0
     layerName=Farm_Grass
     previewWeight=1
     constLayerInput=<0 0 0>
     heightInput {expression=null, inputName=None, mask=0, maskA=0, maskB=0, maskG=0, maskR=0, outputIndex=0}
     layerInput {
      expression=&self:MaterialExpressionMaterialFunctionCall_18
      inputName=None
      mask=0
      maskA=0
      maskB=0
      maskG=0
      maskR=0
      outputIndex=1
     }
    }
    {
     blendType=LB_WeightBlend
     constHeightInput=0
     layerName=Mountain_Grass
     previewWeight=1
     constLayerInput=<0 0 0>
     heightInput {expression=null, inputName=None, mask=0, maskA=0, maskB=0, maskG=0, maskR=0, outputIndex=0}
     layerInput {
      expression=&self:MaterialExpressionMaterialFunctionCall_23
      inputName=None
      mask=0
      maskA=0
      maskB=0
      maskG=0
      maskR=0
      outputIndex=2
     }
    }
   ]}
  n115 = ScalarParameter() {defaultValue=1, parameterName=RockEdgeDesat}
  n116 = ScalarParameter() {defaultValue=1, parameterName=RockEdgeDesat_Arid}
  n117 = LandscapeLayerSample() {parameterName=Arid}
  n118 = Abs(None=ComponentMask(None=VertexNormalWS()) {b=true})
  n119 = Min(
   A=OneMinus(None=@n21.Lerp Alpha No Contrast)
   B=ComponentMask(None=MaterialFunctionCall(In=@n118, Contrast=@n16.Proj_Contrast) {materialFunction=&CheapContrast}.Result) {r=true}
  )
  n120 = ScalarParameter() {defaultValue=-1.5, group=Arid_SandRockBlend, parameterName=Arid_SandRock_LOW_1}
  n121 = ScalarParameter() {defaultValue=1.5, group=Arid_SandRockBlend, parameterName=Arid_SandRock_High_1}
  n122 = ComponentMask(None=VertexNormalWS()) {b=true}
  n123 = Clamp(None=Add(
    A=LinearInterpolate(A=@n120, B=@n121, Alpha=@n122)
    B=ScalarParameter() {defaultValue=-0.5429, group=Arid_SandRockBlend, parameterName=Arid_SandRock_Offset_1}
   ))
  n124 = Min(
   A=OneMinus(None=@n63.Lerp Alpha No Contrast)
   B=ComponentMask(None=MaterialFunctionCall(In=@n123, Contrast=@n43) {materialFunction=&CheapContrast}.Result) {r=true}
  )
  n125 = Desaturation(
   None=@n114
   Fraction=LinearInterpolate(B=LinearInterpolate(A=@n115, B=@n116, Alpha=@n117), Alpha=LinearInterpolate(A=@n119, B=@n124, Alpha=@n117))
  ) {desc=To Render Colormap}
  n126 = LinearInterpolate(A=@n125, B=Constant() {r=1}, Alpha=Clamp(None=Divide(A=@n67, B=@n69)))
  n127 = ScalarParameter() {parameterName=Golf_Mobile_LayerContrast}
  LinearInterpolate(
   A=@n126
   B=@n88.Fairway_Color
   Alpha=MaterialFunctionCall(In=LandscapeLayerSample() {parameterName=Golf_Fairway}, Contrast=@n127) {materialFunction=&CheapContrast}.Result
  )
  MaterialExpressionConstant_6 = Constant()
  MaterialExpressionConstant3Vector_13 = Constant3Vector() {constant=<rgba 0 0 1 0>}
  MaterialExpressionConstant3Vector_14 = Constant3Vector() {desc=Farm, constant=<rgba 1 0 0 0>}
  MaterialExpressionConstant3Vector_15 = Constant3Vector() {desc=Mountain, constant=<rgba 0 1 0 0>}
  n128 = LandscapeLayerBlend(
   Layer Grass_Only=@MaterialExpressionConstant_6
   Layer Grass_Topology=@MaterialExpressionConstant_6
   Layer Arid=@MaterialExpressionConstant3Vector_13
   Layer Farm_Grass=@MaterialExpressionConstant3Vector_14
   Layer Mountain_Grass=@MaterialExpressionConstant3Vector_15
  ) {layers [
    {
     blendType=LB_WeightBlend
     constHeightInput=0
     layerName=Grass_Only
     previewWeight=1
     constLayerInput=<0 0 0>
     heightInput {expression=null, inputName=None, mask=0, maskA=0, maskB=0, maskG=0, maskR=0, outputIndex=0}
     layerInput {
      expression=&self:MaterialExpressionConstant_6
      inputName=None
      mask=0
      maskA=0
      maskB=0
      maskG=0
      maskR=0
      outputIndex=0
     }
    }
    {
     blendType=LB_WeightBlend
     constHeightInput=0
     layerName=Grass_Topology
     previewWeight=1
     constLayerInput=<0 0 0>
     heightInput {expression=null, inputName=None, mask=0, maskA=0, maskB=0, maskG=0, maskR=0, outputIndex=0}
     layerInput {
      expression=&self:MaterialExpressionConstant_6
      inputName=None
      mask=0
      maskA=0
      maskB=0
      maskG=0
      maskR=0
      outputIndex=0
     }
    }
    {
     blendType=LB_WeightBlend
     constHeightInput=0
     layerName=Arid
     previewWeight=1
     constLayerInput=<0 0 0>
     heightInput {expression=null, inputName=None, mask=0, maskA=0, maskB=0, maskG=0, maskR=0, outputIndex=0}
     layerInput {
      expression=&self:MaterialExpressionConstant3Vector_13
      inputName=None
      mask=0
      maskA=0
      maskB=0
      maskG=0
      maskR=0
      outputIndex=0
     }
    }
    {
     blendType=LB_WeightBlend
     constHeightInput=0
     layerName=Farm_Grass
     previewWeight=1
     constLayerInput=<0 0 0>
     heightInput {expression=null, inputName=None, mask=0, maskA=0, maskB=0, maskG=0, maskR=0, outputIndex=0}
     layerInput {
      expression=&self:MaterialExpressionConstant3Vector_14
      inputName=None
      mask=0
      maskA=0
      maskB=0
      maskG=0
      maskR=0
      outputIndex=0
     }
    }
    {
     blendType=LB_WeightBlend
     constHeightInput=0
     layerName=Mountain_Grass
     previewWeight=1
     constLayerInput=<0 0 0>
     heightInput {expression=null, inputName=None, mask=0, maskA=0, maskB=0, maskG=0, maskR=0, outputIndex=0}
     layerInput {
      expression=&self:MaterialExpressionConstant3Vector_15
      inputName=None
      mask=0
      maskA=0
      maskB=0
      maskG=0
      maskR=0
      outputIndex=0
     }
    }
   ]}
  LinearInterpolate(A=@n128)
  MaterialFunctionCall(Transition Phase=@n60, Height Texture=@n61, Contrast=@n62) {desc=Base Color High end, materialFunction=&HeightLerp}
  MaterialFunctionCall(In=@n101.Blended Material, Color=VectorParameter() {parameterName=Gravel_Spline_Match, defaultValue=<rgba 1 0.7689 0.6574 0>}.RGB) {materialFunction=&MF_Athena_Material_BaseMultiply}
  n129 = RuntimeVirtualTextureSample!(UVs, World Position) {
   materialType=BaseColor_Normal_Specular_YCoCg
   virtualTexture=/BlastBerryMapContent/Game/Athena/Environments/Materials/RVT/Athena_RuntimeVirtualTexture
  }
  n130 = MakeMaterialAttributes(
   BaseColor=@n129.BaseColor
   Specular=@n129.BaseColor
   Roughness=@n129.BaseColor
   Normal=Transform(None=@n129.Normal) {transformSourceType=TRANSFORMSOURCE_World, transformType=TRANSFORM_Tangent}
   SubsurfaceColor=Constant()
  )
  n131 = LandscapeLayerSample() {parameterName=Grass_Only}
  n132 = Constant() {r=1}
  n133 = LinearInterpolate(A=BreakMaterialAttributes!(Attr=@n86).Roughness, B=Constant() {r=1})
  n134 = QualitySwitch(Default=@n133, Low=@n132, High=@n133, Medium=@n132)
  n135 = FeatureLevelSwitch(Default=@n134, ES2=@n132, ES3_1=@n132, SM4_REMOVED=@n134, SM5=@n134, SM6=@n134)
  n136 = Divide(
   A=ComponentMask(None=WorldPosition() {worldPositionShaderOffset=WPT_ExcludeAllShaderOffsets}.XYZ) {g=true, r=true}
   B=ScalarParameter() {defaultValue=25000, parameterName=MacroNormalScale}
  )
  n137 = ScalarParameter() {defaultValue=0.625, parameterName=MacroNormalIntensity}
  n138 = Constant() {r=0.25}
  n139 = Constant() {r=0.9}
  n140 = FeatureLevelSwitch(Default=@n138, ES2=@n139, ES3_1=@n139, SM4_REMOVED=@n138)
  n141 = Multiply(
   A=LinearInterpolate(
    A=Constant3Vector() {constant=<rgba 0 0 1 0>}
    B=TextureSample(UVs=@n136) {samplerType=SAMPLERTYPE_Normal, texture=/Game/ContentCreationTools/TextureCreation/Textures/T_BPCreated_MacroNormal_01}.RGB
    Alpha=MaterialFunctionCall!(Distance) {materialFunction=/Game/Athena/Environments/Landscape/MaterialFunctions/MF_TerrainDistanceFade}.Result
   )
   B=MaterialFunctionCall(X=@n137, Y=@n137, Z=QualitySwitch(Default=@n140, Low=@n139, Medium=@n139)) {materialFunction=&MakeFloat3}.Result
  )
  n142 = Add(A=Multiply(A=@n112, B=Constant3Vector() {constant=<rgba 1 1 0.5 0>}), B=@n141)
  n143 = QualitySwitch(Default=@n142, Low=@n141, High=@n142, Medium=@n141)
  n144 = FeatureLevelSwitch(Default=@n143, ES2=@n141, ES3_1=@n141, SM4_REMOVED=@n143, SM5=@n143, SM6=@n143)
  n145 = LandscapeLayerSample() {parameterName=Arid}
  n146 = MaterialFunctionCall!(In=LandscapeLayerCoords() {mappingScale=4}) {materialFunction=/BlastBerryMapContent/Game/Athena/Environments/Landscape/MaterialFunctions/Standard/MF_Athena_Mud_01}
  n147 = LandscapeLayerSwitch(
   LayerUsed=MaterialFunctionCall(Base Material=@n146.Result, Top Material=@n105.Result, Alpha=@n98) {materialFunction=&MatLayerBlend_Standard}.Blended Material
   LayerNotUsed=@n146.Result
  ) {parameterName=Crater}
  n148 = Multiply(
   A=@n146.BaseColor_Plain
   B=ComponentMask(None=CollectionParameter() {collection=&MPC_Landscape, parameterName=Arid_Mud_Tint}) {b=true, g=true, r=true}
  )
  n149 = LandscapeLayerSwitch(
   LayerUsed=MaterialFunctionCall(Material=@n147, NewBaseColor=@n148, Mask=LandscapeLayerSample() {parameterName=Arid}) {materialFunction=&MatLayerBlend_OverrideBaseColor}.Blended Material
   LayerNotUsed=@n147
  ) {parameterName=Arid}
  n150 = OneMinus(None=@MaterialExpressionMaterialFunctionCall_74.Result)
  n151 = MaterialFunctionCall(Mat attributes=@n146.Result) {materialFunction=&MatLayerBlend_BreakOpacity}
  n152 = LandscapeLayerSample() {parameterName=Arid}
  n153 = MaterialFunctionCall(
   Transition Phase=LandscapeLayerSample() {parameterName=Mud}
   Height Texture=LandscapeLayerSwitch(LayerUsed=LinearInterpolate(A=@n150, B=@n151.Result, Alpha=@n152), LayerNotUsed=@n150) {parameterName=Arid}
   Contrast=ScalarParameter() {defaultValue=1, group=Mud, parameterName=Mud_Contrast}
  ) {materialFunction=&HeightLerp}
  n154 = MaterialFunctionCall!(UVs=LandscapeLayerCoords() {mappingScale=4}) {materialFunction=/BlastBerryMapContent/Game/Athena/Environments/Landscape/MaterialFunctions/Standard/MF_Athena_Forest_01}
  n155 = MaterialFunctionCall(
   Base Material=@n90
   Top Material=@n154.Result
   Alpha=MaterialFunctionCall(
    Transition Phase=LandscapeLayerSample() {parameterName=Forest, previewWeight=1}
    Height Texture=@n154.Height
    Contrast=ScalarParameter() {defaultValue=0.25, group=Forest, parameterName=Forest_Contrast}
   ) {materialFunction=&HeightLerp}.Alpha
  ) {materialFunction=&MatLayerBlend_Standard}
  n156 = MaterialFunctionCall!(UseSimple, In=LandscapeLayerCoords() {mappingScale=4}) {materialFunction=/BlastBerryMapContent/Game/Athena/Environments/Landscape/MaterialFunctions/Standard/MF_Athena_Road_01}
  n157 = MaterialFunctionCall(
   Transition Phase=Power(Base=LandscapeLayerSample() {parameterName=Road, previewWeight=1})
   Height Texture=MaterialFunctionCall(Mat attributes=@n156.Result) {materialFunction=&MatLayerBlend_BreakOpacity}.Result
   Contrast=ScalarParameter() {defaultValue=5, group=Road, parameterName=Road_Contrast}
  ) {materialFunction=&HeightLerp}
  n158 = MaterialFunctionCall(
   Transition Phase=Multiply(
    A=LandscapeLayerSample() {parameterName=Road, previewWeight=1}
    B=ScalarParameter() {defaultValue=1.2, group=Road, parameterName=Raod_Transition}
   )
   Height Texture=Multiply(
    A=OneMinus(None=@MaterialExpressionMaterialFunctionCall_74.Result)
    B=ScalarParameter() {defaultValue=0.25, group=Road, parameterName=Road_heightMP_2}
   )
   Contrast=ScalarParameter() {defaultValue=5, group=Road, parameterName=RoadMud_Contrast}
  ) {materialFunction=&HeightLerp}
  n159 = TextureSample(UVs=LandscapeLayerCoords() {mappingScale=130}) {
   samplerSource=SSM_Wrap_WorldGroupSettings
   samplerType=SAMPLERTYPE_LinearColor
   texture=/BlastBerryMapContent/Game/Athena/Environments/Textures/T_Athena_Terrain_Topo_Mask
  }
  n160 = LinearInterpolate(
   A=@n96.Height
   B=Multiply(A=@n159.R, B=ScalarParameter() {defaultValue=0.5, parameterName=HeightDistanceBlend})
   Alpha=MaterialFunctionCall() {materialFunction=&MF_TerrainDistanceFade}.Result
  ) {constB=0.5}
  n161 = MaterialFunctionCall(
   Transition Phase=LandscapeLayerSample() {parameterName=Gravel, previewWeight=1}
   Height Texture=FeatureLevelSwitch(Default=@n160, ES2=@n96.Height, ES3_1=@n96.Height, SM4_REMOVED=@n160, SM5=@n160, SM6=@n160)
   Contrast=ScalarParameter() {group=Gravel, parameterName=Gravel_BlendContrast}
  ) {materialFunction=&HeightLerp}
  n162 = MaterialFunctionCall(
   Base Material=MaterialFunctionCall(
    Base Material=MaterialFunctionCall(Base Material=@n155.Blended Material, Top Material=@n149, Alpha=@n153.Alpha) {materialFunction=&MatLayerBlend_Standard}.Blended Material
    Top Material=MaterialFunctionCall(Base Material=@n149, Top Material=@n156.Result, Alpha=@n157.Alpha) {materialFunction=&MatLayerBlend_Standard}.Blended Material
    Alpha=@n158.Alpha
   ) {materialFunction=&MatLayerBlend_Standard}.Blended Material
   Top Material=@n102
   Alpha=@n161.Alpha
  ) {materialFunction=&MatLayerBlend_Standard}
  n163 = MaterialFunctionCall(Material=@n162.Blended Material) {materialFunction=&MatLayerBlend_BreakBaseColor}
  n164 = Add(A=OneMinus(None=@n22), B=LandscapeLayerSample() {parameterName=Gravel})
  n165 = Add(A=LandscapeLayerSample() {parameterName=Mud}, B=LandscapeLayerSample() {parameterName=Road})
  n166 = OneMinus(None=Saturate(None=Add(A=@n164, B=@n165)))
  n167 = MaterialFunctionCall!(
   Mask 2=TextureObject() {samplerType=SAMPLERTYPE_LinearColor, texture=&T_Athena_Terrain_Topo_Mask}
   UV Coordinates=LandscapeLayerCoords() {mappingScale=1525}
   RockMask
   GrassMask=@n166
   Arid=@n145
   BaseTexture=@n163.BaseColor
  ) {materialFunction=/BlastBerryMapContent/Game/Athena/Environments/Materials/MaterialFunctions/MF_TerrainTopoAdjustment}
  n168 = Saturate(None=CollectionParameter() {collection=&MPC_Landscape, parameterName=DisplayLowDetailShader})
  n169 = Multiply(
   A=Desaturation(
    None=@n167.Result
    Fraction=CollectionParameter() {collection=&MPC_Landscape, parameterName=Arid_EdgeDesaturation}
   )
   B=ComponentMask(None=CollectionParameter() {collection=&MPC_Landscape, parameterName=Arid_SandSediment_Tint}) {b=true, g=true, r=true}
  )
  n170 = LinearInterpolate(
   A=StaticSwitchParameter(True=LinearInterpolate(A=@n167.Result, B=@n163.BaseColor, Alpha=@n168), False=LinearInterpolate(A=@n167.Result, B=@n169, Alpha=@n167.Result)) {group=Utilities, parameterName=EnableQuickColorRender}
   B=@n163.BaseColor
   Alpha=LandscapeLayerSample() {parameterName=NoTopo}
  )
  n171 = LinearInterpolate(
   A=@n126
   B=@n88.Golf_Green_Color
   Alpha=MaterialFunctionCall(In=LandscapeLayerSample() {parameterName=Golf_Green}, Contrast=@n127) {materialFunction=&CheapContrast}.Result
  )
  n172 = MaterialFunctionCall(
   Mask 2=TextureObject() {samplerType=SAMPLERTYPE_LinearColor, texture=&T_Athena_Terrain_Topo_Mask}
   UV Coordinates=LandscapeLayerCoords() {mappingScale=1525}
   GrassMask=@n166
   Arid=@n145
   BaseTexture=@n171
  ) {materialFunction=&MF_TerrainTopoAdjustment}
  n173 = ScalarParameter() {parameterName=Arid_Edge_Desat_1}
  n174 = CollectionParameter() {collection=&MPC_Landscape, parameterName=Arid_SandSediment_Tint}
  n175 = LinearInterpolate(
   A=@n172.Result
   B=Multiply(A=Desaturation(None=@n172.Result, Fraction=@n173), B=ComponentMask(None=@n174) {b=true, g=true, r=true})
   Alpha=@n172.Result
  )
  n176 = LandscapeLayerSample() {parameterName=NoTopo}
  n177 = VertexNormalWS()
  n178 = Desaturation(
   None=Divide(
    A=LinearInterpolate(A=@n175, B=@n171, Alpha=@n176)
    B=CollectionParameter() {collection=&MPC_Landscape, parameterName=ColormapDivisor}
   )
   Fraction=SphereMask(
    A=CollectionParameter() {collection=&MPC_Landscape, parameterName=GrassEdgeZ}
    B=ComponentMask(None=@n177) {b=true}
    Radius=CollectionParameter() {collection=&MPC_Landscape, parameterName=GrassEdgeDesatWidth}
   ) {hardnessPercent=0}
  )
  n179 = CollectionParameter() {collection=&MPC_Landscape, parameterName=RenderColorMap}
  n180 = LinearInterpolate(
   A=LinearInterpolate(
    A=@n170
    B=MaterialFunctionCall!(UVW=@n178) {materialFunction=/Engine/ArtTools/RenderToTexture/MaterialFunctions/MarkRedOutside0-1Range}.Result
    Alpha=Saturate(None=@n179)
   )
   B=@n128
   Alpha=Saturate(None=CollectionParameter() {collection=&MPC_Landscape, parameterName=DisplayBiomeRGB})
  )
  n181 = StaticSwitchParameter(
   True=MaterialFunctionCall!(
    Mat_In=@n162.Blended Material
    BaseColor=StaticSwitchParameter(True=@n180, False=@n170) {group=Utilities, parameterName=EnableQuickColorRender}
   ) {materialFunction=/Game/Athena/Environments/Landscape/MaterialFunctions/MF_Athena_ReplaceBaseColor}.Result
   False=@n162.Blended Material
  ) {group=Rufus, parameterName=UseTopo}
  n182 = MaterialFunctionCall(
   Material=@n77.Result
   NewBaseColor=MaterialFunctionCall(Material=@n154.Result) {materialFunction=&MatLayerBlend_BreakBaseColor}.BaseColor
   Mask=MaterialFunctionCall(In=LandscapeLayerSample() {parameterName=Forest, previewWeight=1}, Contrast=Constant() {r=2}) {materialFunction=&CheapContrast}.Result
  ) {materialFunction=&MatLayerBlend_OverrideBaseColor}
  n183 = MaterialFunctionCall(Material=@n149) {materialFunction=&MatLayerBlend_BreakBaseColor}
  n184 = MaterialFunctionCall(Material=@n182.Blended Material, NewBaseColor=@n183.BaseColor, Mask=@n153.Alpha) {desc=medium, materialFunction=&MatLayerBlend_OverrideBaseColor}
  n185 = MaterialFunctionCall(Material=@n156.Result) {desc=Skip mud blend for Low/Mobile, materialFunction=&MatLayerBlend_BreakBaseColor}
  n186 = MaterialFunctionCall(Material=@n102) {materialFunction=&MatLayerBlend_BreakBaseColor}
  n187 = MaterialFunctionCall(Material=@n182.Blended Material, NewBaseColor=@n183.BaseColor, Mask=LandscapeLayerSample() {parameterName=Mud}) {desc=Low, materialFunction=&MatLayerBlend_OverrideBaseColor}
  n188 = MaterialFunctionCall(
   Material=MaterialFunctionCall(
    Material=FeatureLevelSwitch(Default=@n187.Blended Material, ES2=@n182.Blended Material, ES3_1=@n182.Blended Material)
    NewBaseColor=@n185.BaseColor
    Mask=@n158.Alpha
   ) {materialFunction=&MatLayerBlend_OverrideBaseColor}.Blended Material
   NewBaseColor=@n186.BaseColor
   Mask=LandscapeLayerSample() {parameterName=Gravel, previewWeight=1}
  ) {desc=Low, materialFunction=&MatLayerBlend_OverrideBaseColor}
  n189 = QualitySwitch(
   Default=@n181
   Low=@n188.Blended Material
   High=@n181
   Medium=MaterialFunctionCall(
    Material=MaterialFunctionCall(Material=@n184.Blended Material, NewBaseColor=@n185.BaseColor, Mask=@n158.Alpha) {desc=medium, materialFunction=&MatLayerBlend_OverrideBaseColor}.Blended Material
    NewBaseColor=@n186.BaseColor
    Mask=@n161.Alpha
   ) {desc=medium, materialFunction=&MatLayerBlend_OverrideBaseColor}.Blended Material
   Epic=@n181
  )
  n190 = BreakMaterialAttributes(Attr=FeatureLevelSwitch(Default=@n189, ES2=@n188.Blended Material, ES3_1=@n188.Blended Material, SM4_REMOVED=@n189, SM5=@n189, SM6=@n189))
  n191 = ComponentMask(None=@n190.BaseColor) {r=true}
  n192 = LinearInterpolate(
   A=@n191
   B=Multiply(A=@n191, B=ScalarParameter() {defaultValue=0.5, parameterName=Arid_SpecMP})
   Alpha=LandscapeLayerSample() {parameterName=Arid}
  )
  n193 = Multiply(A=@n191, B=ScalarParameter() {defaultValue=1, parameterName=SimpleMobileSpec})
  n194 = StaticSwitchParameter(True=FeatureLevelSwitch(Default=@n192, ES2=@n193, ES3_1=@n193, SM4_REMOVED=@n192, SM5=@n192, SM6=@n192), False=@n190.Specular) {defaultValue=true, parameterName=UseCheapSpec}
  n195 = MakeMaterialAttributes(BaseColor=@n190.BaseColor, Specular=@n194, Roughness=@n135, Normal=@n144, SubsurfaceColor=Constant())
  n196 = MaterialFunctionCall(
   NoRVT=@n195
   RVT=StaticSwitchParameter(True=BlendMaterialAttributes(A=@n195, B=@n130, Alpha=@n131), False=@n130) {group=Rufus, parameterName=RVT_LayerSetup}
  ) {
   desc=used to check for switch as well
   materialFunction=&FN_RVT_Switch_Material
  }
  MaterialFunctionCall!(In=StaticSwitchParameter(True=@n195, False=@n196.Result) {group=Rufus, parameterName=RVT_Off}) {materialFunction=/Game/Athena/Asteria/Environments/Landscape/Materials/MaterialFunctions/MF_Terrain_LumenModification}
  MaterialFunctionCall!(Material=@n11.Blended Material, NewRoughness=Constant() {r=0.25}, Mask=@n10) {materialFunction=/Engine/Functions/MaterialLayerFunctions/MatLayerBlend_RoughnessOverride}
  MaterialFunctionCall(Mat attributes=@n155.Blended Material) {materialFunction=&MatLayerBlend_BreakOpacity}
  MaterialFunctionCall() {materialFunction=&MatLayerBlend_BreakOpacity}
  MaterialFunctionCall(Material=@n76) {materialFunction=&MatLayerBlend_BreakBaseColor}
  MaterialFunctionCall!(UseVertexInterp) {materialFunction=/Game/Athena/Environments/Landscape/MaterialFunctions/MF_Landscape_GrassDistanceBlend}
  Multiply(A=@n148, B=@n113)
  Multiply()
  Multiply(A=@n125) {desc=BiomeColors}
  QualitySwitch(Default=@n90, Low=@n77.Result, High=@n90, Medium=@n77.Result) {desc=Use Colormap for Low/Mobile - Normal Blends for Medium/High/Epic}
  Reroute(None=@n21.Alpha)
  RuntimeVirtualTextureOutput!(
   BaseColor=@n190.BaseColor
   Specular=@n194
   Roughness=@n135
   Normal=Transform(None=@n144)
   WorldHeight=ComponentMask(None=WorldPosition().XYZ) {b=true}
   Opacity
   Mask
   Displacement
   Mask4
  )
  Saturate()
  ScalarParameter() {parameterName=Arid_Edge_Desat}
  ScalarParameter() {defaultValue=306200, parameterName=TerrainSize}
  ScalarParameter() {defaultValue=1, parameterName=TestingBiomeRender}
  ScalarParameter() {defaultValue=30000, parameterName=Testing_gradient}
  Subtract(A=Subtract(A=@n95, B=LandscapeLayerSample() {parameterName=Gravel}), B=@n94)
  Subtract(A=@MaterialExpressionSubtract_1, B=LandscapeLayerSample() {parameterName=Arid_Grass_Light})
  Subtract(A=@n95)
  TextureSample(UVs=LandscapeLayerCoords() {mappingScale=5}) {samplerSource=SSM_Wrap_WorldGroupSettings, texture=/Game/Environments/AutumnDecay/Terrain/Textures/T_Grass_AD_D}
  VectorParameter() {parameterName=SandColorationTest}
  VectorParameter() {parameterName=SandColorationTest_1}
 ]
}
