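# Frosty landscape master material for the BlastBerry (Rufus) map, blending painted grass, rock, snow, ice, forest, mud, road, gravel, golf and arid layers.
# Slope and height-lerp masks mix material functions, then biome colormap, topo tinting and macro normals add large-scale variation.
# Layer masks drive seven landscape grass types and an ice emissive.
# Quality and feature-level switches provide cheaper low, medium and mobile paths, alongside debug colormap render switches.
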Material /BlastBerryMapContent/Game/Athena/Environments/Materials/M_Rufus_Frosty_Master
graph {
 properties {
  bUseMaterialAttributes=true
  bUsedWithNanite=true
  pixelDepthOffsetMode=PDOM_Legacy
  refractionMethod=RM_IndexOfRefraction
  shadingModel=MSM_FromMaterialExpression
  usedShadingModels=Default Lit | Subsurface
 }
 nodes [
  Constant3Vector() {constant=<rgba 0 0 1 0>}
  Constant3Vector() {desc=Farm, constant=<rgba 1 0 0 0>}
  n1 = NamedRerouteUsage()
  n2 = LandscapeLayerSample() {parameterName=Grass_Topology, previewWeight=1}
  n3 = Round(None=LandscapeLayerSample() {parameterName=Grass_Only, previewWeight=1})
  n4 = ScalarParameter() {defaultValue=-65, parameterName=ShiftGrassEdgeLOW}
  n5 = ScalarParameter() {defaultValue=50, parameterName=ShiftGrassEdgeHigh}
  n6 = MaterialFunctionCall() {materialFunction=/BlastBerryMapContent/Game/Athena/Environments/Landscape/MaterialFunctions/Standard/MF_Athena_Grass_01}
  n7 = MaterialFunctionCall!(height_Offset, height_GradScale, Coords=WorldPosition().XYZ) {materialFunction=/BlastBerryMapContent/Game/Athena/Environments/Materials/MaterialFunctions/MF_Rufus_WorldHeightGrad}
  n8 = Constant()
  n9 = MaterialFunctionCall() {materialFunction=/BlastBerryMapContent/Game/Athena/Environments/Landscape/MaterialFunctions/Standard/MF_Athena_Rock_01}
  n10 = MaterialFunctionCall!(
   In=@n9.Result
   Color=LinearInterpolate(
    A=CollectionParameter() {collection=/BlastBerryMapContent/Game/Athena/Environments/Landscape/MPC/MPC_Landscape, parameterName=Rock_Bottom_Brightness}
    B=CollectionParameter() {collection=&MPC_Landscape, parameterName=Rock_Top_Brightness}
    Alpha=MaterialFunctionCall(Transition Phase=@n7.Result, Height Texture=@n8, Contrast=@n8) {materialFunction=&HeightLerp}.Alpha
   )
  ) {materialFunction=/Game/Athena/Environments/Landscape/MaterialFunctions/MF_Athena_Material_BaseMultiply}
  n11 = Multiply(
   A=Multiply(
    A=MaterialFunctionCall!(Mat attributes=@n6.Result) {materialFunction=/Engine/Functions/Engine_MaterialFunctions02/MatLayerBlend_BreakOpacity}.Result
    B=MaterialFunctionCall(Mat attributes=@n10.Result) {materialFunction=&MatLayerBlend_BreakOpacity}.Result
   )
   B=ScalarParameter() {defaultValue=0.08, group=Grass, parameterName=Transition_Phase}
  )
  n12 = Multiply(A=ComponentMask(None=VertexNormalWS()) {b=true}, B=ScalarParameter() {defaultValue=2, group=Rock, parameterName=RockMultiply_2})
  n13 = ScalarParameter() {defaultValue=5, group=Grass, parameterName=Contrast}
  n14 = MaterialFunctionCall(Transition Phase=@n11, Height Texture=@n12, Contrast=@n13) {materialFunction=&HeightLerp}
  n15 = Multiply(A=@n14.Lerp Alpha No Contrast) {desc=Rock Alpha for Grass system}
  n16 = Add(A=@n15, B=ScalarParameter() {defaultValue=-0.24, parameterName=ShiftGrassEdge_2})
  n17 = OneMinus(None=Clamp(None=LinearInterpolate(A=@n4, B=@n5, Alpha=@n16)))
  n18 = LandscapeLayerSample() {parameterName=Gravel}
  n19 = LandscapeLayerSample() {parameterName=Mud}
  n20 = LandscapeLayerSample() {parameterName=Road, previewWeight=1}
  n21 = Add(A=Add(A=Add(A=@n18, B=@n17) {constB=0}, B=@n19), B=@n20)
  n22 = LandscapeLayerSample() {parameterName=NoFoliage}
  n23 = Add(A=Ceil(None=LandscapeLayerSample() {parameterName=Golf_Fairway}), B=Ceil(None=LandscapeLayerSample() {parameterName=Golf_Green}))
  n24 = Add(A=Add(A=Add(A=@n21, B=@n22), B=LandscapeLayerSample() {parameterName=Forest}), B=Round(None=@n23))
  n25 = Subtract(A=Subtract(A=Add(A=@n2, B=@n3), B=Round(None=@n17)), B=@n24)
  MaterialExpressionSubtract_2 = Subtract(A=@n25, B=@n1)
  n26 = LandscapeLayerSample() {parameterName=Farm_Grass}
  n27 = Subtract(A=Add(A=LandscapeLayerSample() {parameterName=Mountain_Grass}, B=@n26), B=@n24)
  MaterialExpressionSubtract_4 = Subtract(A=@n27, B=@n1)
  n28 = LandscapeLayerSample() {parameterName=Forest}
  n29 = Add(A=Add(A=Add(A=@n19, B=@n20), B=@n23), B=@n22)
  MaterialExpressionSubtract_9 = Subtract(A=Subtract(A=Subtract(A=@n28, B=@n29), B=Round(None=@n17)), B=@n1)
  n30 = LandscapeLayerSample() {parameterName=Gravel}
  MaterialExpressionSubtract_10 = Subtract(A=Subtract(A=Subtract(A=@n30, B=@n17), B=LandscapeLayerSample() {parameterName=NoFoliage}), B=@n1)
  MaterialExpressionSubtract_0 = Subtract(
   A=Subtract(A=LandscapeLayerSample() {parameterName=Flowers, previewWeight=1}, B=LandscapeLayerSample() {parameterName=NoFoliage, previewWeight=1})
   B=NamedRerouteUsage()
  )
  n31 = Add(A=@n25, B=Subtract(A=@n26, B=@n24))
  MaterialExpressionSubtract_1 = Subtract(A=Clamp(None=Add(A=@n31, B=@n27)), B=NamedRerouteUsage())
  MaterialExpressionSubtract_6 = Subtract(A=Max(A=@n25, B=@n27), B=LandscapeLayerSample() {parameterName=NoTallPlant})
  LandscapeGrassOutput!(
   Grass=@MaterialExpressionSubtract_2
   Mountain=@MaterialExpressionSubtract_4
   Forest=@MaterialExpressionSubtract_9
   Gravel=@MaterialExpressionSubtract_10
   Flowers=@MaterialExpressionSubtract_0
   Bushes=@MaterialExpressionSubtract_1
   TallPlant=@MaterialExpressionSubtract_6
  ) {grassTypes [
    Grass {
     grassType=/BlastBerryMapContent/Game/Athena/Environments/Landscape/Foliage/Athena_Foliage_GrassFrosty
     input {
      expression=&self:MaterialExpressionSubtract_2
      inputName=None
      mask=0
      maskA=0
      maskB=0
      maskG=0
      maskR=0
      outputIndex=0
     }
    }
    Mountain {
     grassType=/BlastBerryMapContent/Game/Athena/Environments/Landscape/Foliage/Athena_Foliage_Mountain_GrassFrosty
     input {
      expression=&self:MaterialExpressionSubtract_4
      inputName=None
      mask=0
      maskA=0
      maskB=0
      maskG=0
      maskR=0
      outputIndex=0
     }
    }
    Forest {
     grassType=/BlastBerryMapContent/Game/Athena/Environments/Landscape/Foliage/Athena_Foliage_ForestFrosty
     input {
      expression=&self:MaterialExpressionSubtract_9
      inputName=None
      mask=0
      maskA=0
      maskB=0
      maskG=0
      maskR=0
      outputIndex=0
     }
    }
    Gravel {
     grassType=/Game/Athena/Environments/Landscape/Foliage/Athena_Foliage_Gravel
     input {
      expression=&self:MaterialExpressionSubtract_10
      inputName=None
      mask=0
      maskA=0
      maskB=0
      maskG=0
      maskR=0
      outputIndex=0
     }
    }
    Flowers {
     grassType=/BlastBerryMapContent/Game/Athena/Environments/Landscape/Foliage/Rufus_Foliage_Flowers
     input {
      expression=&self:MaterialExpressionSubtract_0
      inputName=None
      mask=0
      maskA=0
      maskB=0
      maskG=0
      maskR=0
      outputIndex=0
     }
    }
    Bushes {
     grassType=/BlastBerryMapContent/Game/Athena/Environments/Landscape/Foliage/Athena_Foliage_Bushes
     input {
      expression=&self:MaterialExpressionSubtract_1
      inputName=None
      mask=0
      maskA=0
      maskB=0
      maskG=0
      maskR=0
      outputIndex=0
     }
    }
    TallPlant {
     grassType=/BlastBerryMapContent/Game/Athena/Environments/Landscape/Foliage/Rufus_TallPlant
     input {
      expression=&self:MaterialExpressionSubtract_6
      inputName=None
      mask=0
      maskA=0
      maskB=0
      maskG=0
      maskR=0
      outputIndex=0
     }
    }
   ]}
  LandscapeLayerSample() {parameterName=Arid}
  MaterialExpressionReroute_24 = Reroute(None=@n6.BaseColor_Plain)
  MaterialExpressionReroute_8 = Reroute(None=LinearInterpolate(A=Multiply(A=@n9.BaseColor_Plain), B=@n6.BaseColor_Plain, Alpha=@n14.Alpha))
  MaterialExpressionNamedRerouteUsage_3 = NamedRerouteUsage()
  MaterialExpressionNamedRerouteUsage_5 = NamedRerouteUsage()
  n32 = LandscapeLayerBlend!(
   Layer Grass_Only=@MaterialExpressionReroute_24
   Layer Grass_Topology=@MaterialExpressionReroute_8
   Layer Mountain_Grass=@MaterialExpressionReroute_8
   Layer Snow=@MaterialExpressionNamedRerouteUsage_3
   Layer Ice=@MaterialExpressionNamedRerouteUsage_5
  ) {layers [
    {
     blendType=LB_WeightBlend
     constHeightInput=0
     layerName=Grass_Only
     previewWeight=1
     constLayerInput=<0 0 0>
     heightInput {expression=null, inputName=None, mask=0, maskA=0, maskB=0, maskG=0, maskR=0, outputIndex=0}
     layerInput {
      expression=&self:MaterialExpressionReroute_24
      inputName=None
      mask=0
      maskA=0
      maskB=0
      maskG=0
      maskR=0
      outputIndex=0
     }
    }
    {
     blendType=LB_WeightBlend
     constHeightInput=0
     layerName=Grass_Topology
     previewWeight=1
     constLayerInput=<0 0 0>
     heightInput {expression=null, inputName=None, mask=0, maskA=0, maskB=0, maskG=0, maskR=0, outputIndex=0}
     layerInput {
      expression=&self:MaterialExpressionReroute_8
      inputName=None
      mask=0
      maskA=0
      maskB=0
      maskG=0
      maskR=0
      outputIndex=0
     }
    }
    {
     blendType=LB_WeightBlend
     constHeightInput=0
     layerName=Mountain_Grass
     previewWeight=1
     constLayerInput=<0 0 0>
     heightInput {expression=null, inputName=None, mask=0, maskA=0, maskB=0, maskG=0, maskR=0, outputIndex=0}
     layerInput {
      expression=&self:MaterialExpressionReroute_8
      inputName=None
      mask=0
      maskA=0
      maskB=0
      maskG=0
      maskR=0
      outputIndex=0
     }
    }
    {
     blendType=LB_WeightBlend
     constHeightInput=0
     layerName=Snow
     previewWeight=0
     constLayerInput=<0 0 0>
     heightInput {expression=null, inputName=None, mask=0, maskA=0, maskB=0, maskG=0, maskR=0, outputIndex=0}
     layerInput {
      expression=&self:MaterialExpressionNamedRerouteUsage_3
      inputName=None
      mask=0
      maskA=0
      maskB=0
      maskG=0
      maskR=0
      outputIndex=0
     }
    }
    {
     blendType=LB_WeightBlend
     constHeightInput=0
     layerName=Ice
     previewWeight=0
     constLayerInput=<0 0 0>
     heightInput {expression=null, inputName=None, mask=0, maskA=0, maskB=0, maskG=0, maskR=0, outputIndex=0}
     layerInput {
      expression=&self:MaterialExpressionNamedRerouteUsage_5
      inputName=None
      mask=0
      maskA=0
      maskB=0
      maskG=0
      maskR=0
      outputIndex=0
     }
    }
   ]}
  n33 = ComponentMask(None=CollectionParameter() {collection=&MPC_Landscape, parameterName=Arid_Grass_Tint}) {b=true, g=true, r=true}
  n34 = LandscapeLayerSample() {parameterName=Arid_Grass_Light}
  MaterialExpressionLinearInterpolate_52 = LinearInterpolate(A=ScalarParameter() {defaultValue=0.8, group=Grass, parameterName=Grass_Tint}, Alpha=LandscapeLayerSample() {parameterName=NoTopo})
  n35 = MaterialFunctionCall(In=@n6.Result, Color=@MaterialExpressionLinearInterpolate_52) {materialFunction=&MF_Athena_Material_BaseMultiply}
  n36 = MaterialFunctionCall(
   Transition Phase=@n34
   Height Texture=MaterialFunctionCall(Mat attributes=@n35.Result) {materialFunction=&MatLayerBlend_BreakOpacity}.Result
   Contrast=ScalarParameter() {parameterName=Arid_Grass_Contrast}
  ) {materialFunction=&HeightLerp}
  n37 = ScalarParameter() {defaultValue=1.5, parameterName=AridGrassMob_MP}
  n38 = Clamp(None=Multiply(A=@n34, B=@n37))
  n39 = QualitySwitch(Default=@n36.Alpha, Low=@n38, Medium=@n36.Alpha) {desc="Use Height lerp for Medium, but simple lerp for Mobile/Low"}
  LinearInterpolate(
   A=@n32
   B=Multiply(A=@n33, B=@MaterialExpressionReroute_24)
   Alpha=FeatureLevelSwitch(Default=@n39, ES2=@n38, ES3_1=@n38, SM4_REMOVED=@n39, SM5=@n39, SM6=@n39)
  )
  n40 = Constant()
  n41 = Constant()
  n42 = Multiply(
   A=MaterialFunctionCall() {materialFunction=/BlastBerryMapContent/Game/Athena/Environments/Materials/MaterialFunctions/MF_Rufus_SharedWinter}.IceColor
   B=NamedRerouteUsage()
  )
  n43 = Multiply(
   A=Constant3Vector() {constant=<rgba 0.3346 0.4223 0.5 0>}
   B=Clamp(None=MaterialFunctionCall!(UseLandscapeCoordsOnSwitch) {materialFunction=/BlastBerryMapContent/Game/Athena/Environments/Materials/MaterialFunctions/MF_Athena_Snow_01}.NormalDotCamera)
  )
  n44 = LandscapeLayerSample() {parameterName=Gravel}
  n45 = LandscapeLayerSample() {parameterName=Forest}
  n46 = Subtract(A=LandscapeLayerSample() {parameterName=Ice}, B=Add(A=Add(A=@n44, B=@n45), B=LandscapeLayerSample() {parameterName=Mud})) {desc=Mask out paths and dirt from emmisive}
  n47 = Divide(A=LinearInterpolate(A=@n41, B=Multiply(A=@n42, B=@n43) {constB=0.5}, Alpha=Clamp(None=@n46)), B=EyeAdaptation().EyeAdaptation)
  n48 = ShadingPathSwitch(
   Default=@n40
   Deferred=QualitySwitch(Default=StaticSwitchParameter(True=@n41, False=@n47) {group=Utilities, parameterName=EnableQuickColorRender}, Low=@n40)
  )
  n49 = Constant() {r=1}
  n50 = Add(
   A=Add(A=LandscapeLayerSample() {parameterName=Road, previewWeight=1}, B=LandscapeLayerSample() {parameterName=Mud, previewWeight=1})
   B=LandscapeLayerSample() {parameterName=Gravel, previewWeight=1}
  )
  n51 = LandscapeLayerSample() {parameterName=Forest}
  MaterialExpressionReroute_44 = Reroute(None=@n35.Result)
  MaterialExpressionNamedRerouteUsage_9 = NamedRerouteUsage()
  n52 = ComponentMask(None=CollectionParameter() {collection=&MPC_Landscape, parameterName=RockTint}) {b=true, g=true, r=true}
  n53 = MaterialFunctionCall!(Base Material=@n10.Result, Top Material=@n6.Result, Alpha=@n14.Alpha) {materialFunction=/Engine/Functions/MaterialLayerFunctions/MatLayerBlend_Standard}
  MaterialExpressionReroute_39 = Reroute(None=MaterialFunctionCall(In=@n53.Blended Material, Color=LinearInterpolate(A=@n52, B=@MaterialExpressionLinearInterpolate_52, Alpha=@n14.Alpha)) {materialFunction=&MF_Athena_Material_BaseMultiply}.Result)
  n54 = MaterialFunctionCall(
   Transition Phase=Add(
    A=CollectionParameter() {collection=&MPC_Landscape, parameterName=ColormapTransitionBias}
    B=@n11
   )
   Height Texture=@n12
   Contrast=@n13
  ) {
   desc=Height lerp for ColorMap Texture - has ColormapTransitionBias
   materialFunction=&HeightLerp
  }
  MaterialExpressionMaterialFunctionCall_23 = MaterialFunctionCall!(Rock_Diffuse, ColorBlend_Alpha=@n14.Alpha, ColorBlend_Alpha_Bias=@n54.Alpha, Rock_Diffuse1, Grass_Diffuse) {materialFunction=/BlastBerryMapContent/Game/Athena/Environments/Landscape/MaterialFunctions/MF_MountainGrass_Colors}
  MaterialExpressionReroute_43 = Reroute(None=MaterialFunctionCall(In=@n53.Blended Material, Color=@MaterialExpressionMaterialFunctionCall_23.Blended_Colors) {materialFunction=&MF_Athena_Material_BaseMultiply}.Result)
  MaterialExpressionNamedRerouteUsage_0 = NamedRerouteUsage()
  MaterialExpressionNamedRerouteUsage_1 = NamedRerouteUsage()
  MaterialExpressionNamedRerouteUsage_2 = NamedRerouteUsage()
  n55 = LandscapeLayerBlend!(
   Layer Grass_Only=@MaterialExpressionReroute_44
   Height Grass_Only=@MaterialExpressionNamedRerouteUsage_9
   Layer Grass_Topology=@MaterialExpressionReroute_39
   Layer Mountain_Grass=@MaterialExpressionReroute_43
   Layer Ice=@MaterialExpressionNamedRerouteUsage_0
   Layer Snow=@MaterialExpressionNamedRerouteUsage_1
   Height Snow=@MaterialExpressionNamedRerouteUsage_2
  ) {layers [
    {
     blendType=LB_HeightBlend
     constHeightInput=0
     layerName=Grass_Only
     previewWeight=1
     constLayerInput=<0 0 0>
     heightInput {
      expression=&self:MaterialExpressionNamedRerouteUsage_9
      inputName=None
      mask=0
      maskA=0
      maskB=0
      maskG=0
      maskR=0
      outputIndex=0
     }
     layerInput {
      expression=&self:MaterialExpressionReroute_44
      inputName=None
      mask=0
      maskA=0
      maskB=0
      maskG=0
      maskR=0
      outputIndex=0
     }
    }
    {
     blendType=LB_WeightBlend
     constHeightInput=0
     layerName=Grass_Topology
     previewWeight=1
     constLayerInput=<0 0 0>
     heightInput {expression=null, inputName=None, mask=0, maskA=0, maskB=0, maskG=0, maskR=0, outputIndex=0}
     layerInput {
      expression=&self:MaterialExpressionReroute_39
      inputName=None
      mask=0
      maskA=0
      maskB=0
      maskG=0
      maskR=0
      outputIndex=0
     }
    }
    {
     blendType=LB_WeightBlend
     constHeightInput=0
     layerName=Mountain_Grass
     previewWeight=1
     constLayerInput=<0 0 0>
     heightInput {expression=null, inputName=None, mask=0, maskA=0, maskB=0, maskG=0, maskR=0, outputIndex=0}
     layerInput {
      expression=&self:MaterialExpressionReroute_43
      inputName=None
      mask=0
      maskA=0
      maskB=0
      maskG=0
      maskR=0
      outputIndex=0
     }
    }
    {
     blendType=LB_WeightBlend
     constHeightInput=0
     layerName=Ice
     previewWeight=0
     constLayerInput=<0 0 0>
     heightInput {expression=null, inputName=None, mask=0, maskA=0, maskB=0, maskG=0, maskR=0, outputIndex=0}
     layerInput {
      expression=&self:MaterialExpressionNamedRerouteUsage_0
      inputName=None
      mask=0
      maskA=0
      maskB=0
      maskG=0
      maskR=0
      outputIndex=0
     }
    }
    {
     blendType=LB_HeightBlend
     constHeightInput=0
     layerName=Snow
     previewWeight=0
     constLayerInput=<0 0 0>
     heightInput {
      expression=&self:MaterialExpressionNamedRerouteUsage_2
      inputName=None
      mask=0
      maskA=0
      maskB=0
      maskG=0
      maskR=0
      outputIndex=0
     }
     layerInput {
      expression=&self:MaterialExpressionNamedRerouteUsage_1
      inputName=None
      mask=0
      maskA=0
      maskB=0
      maskG=0
      maskR=0
      outputIndex=0
     }
    }
   ]}
  n56 = LinearInterpolate(A=BreakMaterialAttributes!(Attr=@n55).Roughness, B=Constant() {r=1}, Alpha=Saturate(None=Add(A=@n50, B=@n51)))
  n57 = QualitySwitch(Default=@n56, Low=@n49, High=@n56, Medium=@n49)
  n58 = FeatureLevelSwitch(Default=@n57, ES2=@n49, ES3_1=@n49, SM4_REMOVED=@n57, SM5=@n57, SM6=@n57)
  n59 = Constant3Vector() {constant=<rgba 0 0 1 0>}
  n60 = Clamp(None=@n50)
  n61 = MaterialFunctionCall(
   Base Material=@n55
   Top Material=MaterialFunctionCall(In=@n35.Result, Color=@n33) {materialFunction=&MF_Athena_Material_BaseMultiply}.Result
   Alpha=@n36.Alpha
  ) {materialFunction=&MatLayerBlend_Standard}
  n62 = MaterialFunctionCall!(Material=@n61.Blended Material) {materialFunction=/Engine/Functions/MaterialLayerFunctions/MatLayerBlend_BreakNormal}
  n63 = ScalarParameter() {defaultValue=1, parameterName=Green_NormalFlatten}
  n64 = LandscapeLayerSample() {parameterName=Golf_Fairway, previewWeight=1}
  n65 = MaterialFunctionCall!(In_LowMobile_Material, Golf_Fairway_Mask=@n64, In=@n61.Blended Material) {materialFunction=/BlastBerryMapContent/Game/Athena/Environments/Materials/MaterialFunctions/MF_LawnGrassColoration}
  n66 = LandscapeLayerSwitch!(
   LayerUsed=LinearInterpolate(
    A=@n62.Normal
    B=MaterialFunctionCall(Normal=@n62.Normal, Flatness=@n63) {materialFunction=&FlattenNormal}.Result
    Alpha=@n65.Green_Mask
   )
   LayerNotUsed=@n62.Normal
  ) {parameterName=Golf_Fairway}
  n67 = Divide(
   A=ComponentMask(None=WorldPosition() {worldPositionShaderOffset=WPT_ExcludeAllShaderOffsets}.XYZ) {g=true, r=true}
   B=ScalarParameter() {defaultValue=25000, parameterName=MacroNormalScale}
  )
  n68 = ScalarParameter() {defaultValue=0.625, parameterName=MacroNormalIntensity}
  n69 = Constant() {r=0.25}
  n70 = Constant() {r=0.9}
  n71 = FeatureLevelSwitch(Default=@n69, ES2=@n70, ES3_1=@n70, SM4_REMOVED=@n69)
  n72 = Multiply(
   A=LinearInterpolate(
    A=Constant3Vector() {constant=<rgba 0 0 1 0>}
    B=TextureSample(UVs=@n67) {samplerType=SAMPLERTYPE_Normal, texture=/Game/ContentCreationTools/TextureCreation/Textures/T_BPCreated_MacroNormal_01}.RGB
    Alpha=MaterialFunctionCall!(Distance) {materialFunction=/Game/Athena/Environments/Landscape/MaterialFunctions/MF_TerrainDistanceFade}.Result
   )
   B=MaterialFunctionCall(X=@n68, Y=@n68, Z=QualitySwitch(Default=@n71, Low=@n70, Medium=@n70)) {materialFunction=&MakeFloat3}.Result
  )
  n73 = Add(A=Multiply(A=LinearInterpolate(A=@n66, B=@n59, Alpha=@n60), B=Constant3Vector() {constant=<rgba 1 1 0.5 0>}), B=@n72)
  n74 = MaterialFunctionCall() {materialFunction=&MF_Athena_Snow_01}
  n75 = LandscapeLayerSample() {parameterName=Ice}
  n76 = LandscapeLayerSample() {parameterName=Snow}
  n77 = Add(
   A=Add(A=LandscapeLayerSample() {parameterName=Gravel}, B=LandscapeLayerSample() {parameterName=Forest})
   B=LandscapeLayerSample() {parameterName=Mud}
  )
  n78 = LandscapeLayerSample() {parameterName=Road}
  n79 = Multiply(A=NamedRerouteUsage(), B=Subtract(A=Add(A=@n75, B=@n76), B=Add(A=@n77, B=@n78)) {desc=Mask out paths and dirt from emmisive})
  n80 = QualitySwitch(Default=@n73, Low=@n72, High=@n73, Medium=LinearInterpolate(A=Add(A=@n74.Normal, B=@n72), B=@n74.Normal, Alpha=Clamp(None=@n79)))
  n81 = FeatureLevelSwitch(Default=@n80, ES2=@n72, ES3_1=@n72, SM4_REMOVED=@n80, SM5=@n80, SM6=@n80)
  n82 = MaterialFunctionCall(Mat attributes=@n6.Result) {materialFunction=&MatLayerBlend_BreakOpacity}
  n83 = MaterialFunctionCall(
   Transition Phase=LandscapeLayerSample() {parameterName=Mud}
   Height Texture=OneMinus(None=@n82.Result)
   Contrast=ScalarParameter() {defaultValue=1, group=Mud, parameterName=Mud_Contrast}
  ) {materialFunction=&HeightLerp}
  n84 = MaterialFunctionCall!(UVs=LandscapeLayerCoords() {mappingScale=4}) {materialFunction=/BlastBerryMapContent/Game/Athena/Environments/Landscape/MaterialFunctions/Standard/MF_Athena_Forest_01}
  n85 = ComponentMask(None=WorldPosition().XYZ) {g=true, r=true}
  n86 = ScalarParameter() {defaultValue=306000, parameterName=LandscapeSize, sortPriority=-153000}
  n87 = ScalarParameter() {defaultValue=-153000, parameterName=LandscapeMin, sortPriority=-153000}
  n88 = TextureSampleParameter2D(UVs=Subtract(A=Divide(A=@n85, B=@n86), B=Divide(A=@n87, B=@n86))) {
   parameterName=BiomeColormap
   samplerType=SAMPLERTYPE_LinearColor
   texture=/BlastBerryMapContent/Game/Athena/Environments/Landscape/Textures/T_Athena_Terrain_CombinedColors_01
  }
  n89 = CollectionParameter() {collection=&MPC_Landscape, parameterName=ColormapDivisor}
  n90 = MakeMaterialAttributes!(
   BaseColor=@n32
   Metallic
   Specular
   Roughness
   Anisotropy
   EmissiveColor
   Opacity
   OpacityMask
   Normal
   Tangent
   WorldPositionOffset
   SubsurfaceColor
   ClearCoat
   ClearCoatRoughness
   AmbientOcclusion
   Refraction
   CustomizedUVs_0
   CustomizedUVs_1
   CustomizedUVs_2
   CustomizedUVs_3
   CustomizedUVs_4
   CustomizedUVs_5
   CustomizedUVs_6
   CustomizedUVs_7
   PixelDepthOffset
   ShadingModel
   Displacement
  )
  n91 = MaterialFunctionCall(In_LowMobile_Material=@n90, Golf_Fairway_Mask=@n64) {
   desc=Simple version of fairway for LOW/Medium
   materialFunction=&MF_LawnGrassColoration
  }
  n92 = LandscapeLayerSwitch(LayerUsed=@n91.LowMobile_Fairway, LayerNotUsed=@n90) {parameterName=Golf_Fairway}
  n93 = MaterialFunctionCall!(
   Material=MaterialFunctionCall(In=@n92, Color=Multiply(A=@n88.RGB, B=@n89)) {materialFunction=&MF_Athena_Material_BaseMultiply}.Result
   NewBaseColor=MaterialFunctionCall!(Material=@n92) {materialFunction=/Engine/Functions/MaterialLayerFunctions/MatLayerBlend_BreakBaseColor}.BaseColor
   Mask=NamedRerouteUsage()
  ) {materialFunction=/Engine/Functions/MaterialLayerFunctions/MatLayerBlend_OverrideBaseColor}
  n94 = MaterialFunctionCall(Material=@n93.Blended Material) {materialFunction=&MatLayerBlend_BreakBaseColor}
  n95 = Saturate(None=CollectionParameter() {collection=&MPC_Landscape, parameterName=DisplayLowDetailShader})
  n96 = LandscapeLayerSwitch(LayerUsed=@n65.Result, LayerNotUsed=@n61.Blended Material) {parameterName=Golf_Fairway}
  n97 = MaterialFunctionCall(
   Base Material=StaticSwitchParameter(
    True=MaterialFunctionCall(Material=@n96, NewBaseColor=@n94.BaseColor, Mask=@n95) {materialFunction=&MatLayerBlend_OverrideBaseColor}.Blended Material
    False=@n96
   ) {group=Utilities, parameterName=EnableQuickColorRender}
   Top Material=@n84.Result
   Alpha=MaterialFunctionCall(
    Transition Phase=LandscapeLayerSample() {parameterName=Forest, previewWeight=1}
    Height Texture=@n84.Height
    Contrast=ScalarParameter() {defaultValue=0.25, group=Forest, parameterName=Forest_Contrast}
   ) {materialFunction=&HeightLerp}.Alpha
  ) {materialFunction=&MatLayerBlend_Standard}
  n98 = MaterialFunctionCall!(In=LandscapeLayerCoords() {mappingScale=4}) {materialFunction=/BlastBerryMapContent/Game/Athena/Environments/Landscape/MaterialFunctions/Standard/MF_Athena_Mud_01}
  n99 = MaterialFunctionCall!(UseSimple, In=LandscapeLayerCoords() {mappingScale=4}) {materialFunction=/BlastBerryMapContent/Game/Athena/Environments/Landscape/MaterialFunctions/Standard/MF_Athena_Road_01}
  n100 = MaterialFunctionCall(
   Transition Phase=Power(Base=LandscapeLayerSample() {parameterName=Road, previewWeight=1})
   Height Texture=MaterialFunctionCall(Mat attributes=@n99.Result) {materialFunction=&MatLayerBlend_BreakOpacity}.Result
   Contrast=ScalarParameter() {defaultValue=5, group=Road, parameterName=Road_Contrast}
  ) {materialFunction=&HeightLerp}
  n101 = MaterialFunctionCall(
   Transition Phase=Multiply(
    A=LandscapeLayerSample() {parameterName=Road, previewWeight=1}
    B=ScalarParameter() {defaultValue=1.2, group=Road, parameterName=Raod_Transition}
   )
   Height Texture=Multiply(A=OneMinus(None=@n82.Result), B=ScalarParameter() {defaultValue=0.25, group=Road, parameterName=Road_heightMP_2})
   Contrast=ScalarParameter() {defaultValue=5, group=Road, parameterName=RoadMud_Contrast}
  ) {materialFunction=&HeightLerp}
  n102 = MaterialFunctionCall!(In=LandscapeLayerCoords() {mappingScale=4}) {materialFunction=/BlastBerryMapContent/Game/Athena/Environments/Landscape/MaterialFunctions/Standard/MF_Athena_Gravel_01}
  n103 = CollectionParameter() {collection=&MPC_Landscape, parameterName=Winter_Gravel_Tint}
  n104 = LandscapeLayerSample() {parameterName=Snow}
  n105 = LandscapeLayerSample() {parameterName=Ice}
  n106 = MaterialFunctionCall(
   Material=@n102.Result
   NewBaseColor=Multiply(A=@n102.Plain_BaseColor, B=ComponentMask(None=@n103) {b=true, g=true, r=true})
   Mask=Clamp(None=Add(A=@n104, B=@n105))
  ) {desc=snow blending, materialFunction=&MatLayerBlend_OverrideBaseColor}
  n107 = LandscapeLayerSwitch(
   LayerUsed=@n106.Blended Material
   LayerNotUsed=LandscapeLayerSwitch(LayerUsed=@n106.Blended Material, LayerNotUsed=@n102.Result) {parameterName=Snow}
  ) {parameterName=Ice}
  n108 = TextureSample(UVs=LandscapeLayerCoords() {mappingScale=130}) {
   samplerSource=SSM_Wrap_WorldGroupSettings
   samplerType=SAMPLERTYPE_LinearColor
   texture=/BlastBerryMapContent/Game/Athena/Environments/Textures/T_Athena_Terrain_Topo_Mask
  }
  n109 = LinearInterpolate(
   A=@n102.Height
   B=Multiply(A=@n108.R, B=ScalarParameter() {defaultValue=0.5, parameterName=HeightDistanceBlend})
   Alpha=MaterialFunctionCall() {materialFunction=&MF_TerrainDistanceFade}.Result
  ) {constB=0.5}
  n110 = MaterialFunctionCall(
   Transition Phase=LandscapeLayerSample() {parameterName=Gravel, previewWeight=1}
   Height Texture=FeatureLevelSwitch(Default=@n109, ES2=@n102.Height, ES3_1=@n102.Height, SM4_REMOVED=@n109, SM5=@n109, SM6=@n109)
   Contrast=ScalarParameter() {group=Gravel, parameterName=Gravel_BlendContrast}
  ) {materialFunction=&HeightLerp}
  n111 = MaterialFunctionCall(
   Base Material=MaterialFunctionCall(
    Base Material=MaterialFunctionCall(Base Material=@n97.Blended Material, Top Material=@n98.Result, Alpha=@n83.Alpha) {materialFunction=&MatLayerBlend_Standard}.Blended Material
    Top Material=MaterialFunctionCall(Base Material=@n98.Result, Top Material=@n99.Result, Alpha=@n100.Alpha) {materialFunction=&MatLayerBlend_Standard}.Blended Material
    Alpha=@n101.Alpha
   ) {materialFunction=&MatLayerBlend_Standard}.Blended Material
   Top Material=@n107
   Alpha=@n110.Alpha
  ) {materialFunction=&MatLayerBlend_Standard}
  n112 = MaterialFunctionCall(Material=@n111.Blended Material) {materialFunction=&MatLayerBlend_BreakBaseColor}
  n113 = Add(A=OneMinus(None=@n15), B=LandscapeLayerSample() {parameterName=Gravel})
  n114 = Add(A=LandscapeLayerSample() {parameterName=Mud}, B=LandscapeLayerSample() {parameterName=Road})
  n115 = OneMinus(None=Saturate(None=Add(A=@n113, B=@n114)))
  n116 = MaterialFunctionCall!(
   Mask 2=TextureObject() {samplerType=SAMPLERTYPE_LinearColor, texture=&T_Athena_Terrain_Topo_Mask}
   UV Coordinates=LandscapeLayerCoords() {mappingScale=1525}
   RockMask
   GrassMask=@n115
   Arid
   BaseTexture=@n112.BaseColor
  ) {materialFunction=/BlastBerryMapContent/Game/Athena/Environments/Materials/MaterialFunctions/MF_TerrainTopoAdjustment}
  n117 = Saturate(None=CollectionParameter() {collection=&MPC_Landscape, parameterName=DisplayLowDetailShader})
  n118 = Multiply(
   A=Desaturation(
    None=@n116.Result
    Fraction=CollectionParameter() {collection=&MPC_Landscape, parameterName=Arid_EdgeDesaturation}
   )
   B=ComponentMask(None=CollectionParameter() {collection=&MPC_Landscape, parameterName=Arid_SandSediment_Tint}) {b=true, g=true, r=true}
  )
  n119 = LinearInterpolate(
   A=StaticSwitchParameter(True=LinearInterpolate(A=@n116.Result, B=@n112.BaseColor, Alpha=@n117), False=LinearInterpolate(A=@n116.Result, B=@n118, Alpha=@n116.Result)) {group=Utilities, parameterName=EnableQuickColorRender}
   B=@n112.BaseColor
   Alpha=LandscapeLayerSample() {parameterName=NoTopo}
  )
  MaterialExpressionLinearInterpolate_56 = LinearInterpolate(A=@n52, B=@MaterialExpressionLinearInterpolate_52, Alpha=@n54.Alpha)
  MaterialExpressionNamedRerouteUsage_4 = NamedRerouteUsage()
  MaterialExpressionNamedRerouteUsage_6 = NamedRerouteUsage()
  n120 = LandscapeLayerBlend!(
   Layer Grass_Only=@MaterialExpressionLinearInterpolate_52
   Layer Grass_Topology=@MaterialExpressionLinearInterpolate_56
   Layer Mountain_Grass=@MaterialExpressionMaterialFunctionCall_23.Blended_Colors_Bias
   Layer Snow=@MaterialExpressionNamedRerouteUsage_4
   Layer Ice=@MaterialExpressionNamedRerouteUsage_6
  ) {layers [
    {
     blendType=LB_WeightBlend
     constHeightInput=0
     layerName=Grass_Only
     previewWeight=1
     constLayerInput=<0 0 0>
     heightInput {expression=null, inputName=None, mask=0, maskA=0, maskB=0, maskG=0, maskR=0, outputIndex=0}
     layerInput {
      expression=&self:MaterialExpressionLinearInterpolate_52
      inputName=None
      mask=0
      maskA=0
      maskB=0
      maskG=0
      maskR=0
      outputIndex=0
     }
    }
    {
     blendType=LB_WeightBlend
     constHeightInput=0
     layerName=Grass_Topology
     previewWeight=1
     constLayerInput=<0 0 0>
     heightInput {expression=null, inputName=None, mask=0, maskA=0, maskB=0, maskG=0, maskR=0, outputIndex=0}
     layerInput {
      expression=&self:MaterialExpressionLinearInterpolate_56
      inputName=None
      mask=0
      maskA=0
      maskB=0
      maskG=0
      maskR=0
      outputIndex=0
     }
    }
    {
     blendType=LB_WeightBlend
     constHeightInput=0
     layerName=Mountain_Grass
     previewWeight=1
     constLayerInput=<0 0 0>
     heightInput {expression=null, inputName=None, mask=0, maskA=0, maskB=0, maskG=0, maskR=0, outputIndex=0}
     layerInput {
      expression=&self:MaterialExpressionMaterialFunctionCall_23
      inputName=None
      mask=0
      maskA=0
      maskB=0
      maskG=0
      maskR=0
      outputIndex=2
     }
    }
    {
     blendType=LB_WeightBlend
     constHeightInput=0
     layerName=Snow
     previewWeight=0
     constLayerInput=<0 0 0>
     heightInput {expression=null, inputName=None, mask=0, maskA=0, maskB=0, maskG=0, maskR=0, outputIndex=0}
     layerInput {
      expression=&self:MaterialExpressionNamedRerouteUsage_4
      inputName=None
      mask=0
      maskA=0
      maskB=0
      maskG=0
      maskR=0
      outputIndex=0
     }
    }
    {
     blendType=LB_WeightBlend
     constHeightInput=0
     layerName=Ice
     previewWeight=0
     constLayerInput=<0 0 0>
     heightInput {expression=null, inputName=None, mask=0, maskA=0, maskB=0, maskG=0, maskR=0, outputIndex=0}
     layerInput {
      expression=&self:MaterialExpressionNamedRerouteUsage_6
      inputName=None
      mask=0
      maskA=0
      maskB=0
      maskG=0
      maskR=0
      outputIndex=0
     }
    }
   ]}
  n121 = LandscapeLayerSample() {parameterName=Arid}
  n122 = OneMinus(None=@n14.Lerp Alpha No Contrast)
  n123 = ComponentMask(None=VertexNormalWS()) {b=true}
  n124 = ComponentMask(None=MaterialFunctionCall(In=Abs(None=@n123), Contrast=@n9.Proj_Contrast) {materialFunction=&CheapContrast}.Result) {r=true}
  n125 = VertexNormalWS()
  n126 = Add(
   A=LinearInterpolate(
    A=ScalarParameter() {defaultValue=-1.5, group=Arid_SandRockBlend, parameterName=Arid_SandRock_LOW}
    B=ScalarParameter() {defaultValue=1.5, group=Arid_SandRockBlend, parameterName=Arid_SandRock_High}
    Alpha=ComponentMask(None=@n125) {b=true}
   )
   B=ScalarParameter() {defaultValue=-0.5429, group=Arid_SandRockBlend, parameterName=Arid_SandRock_Offset}
  )
  n127 = ScalarParameter() {defaultValue=10, group=Arid_Rock, parameterName=Arid_Rock_ProjContrast}
  n128 = MaterialFunctionCall!(ProjContrast=@n127) {materialFunction=/BlastBerryMapContent/Game/Athena/Environments/Landscape/MaterialFunctions/Arid/MF_Athena_Arid_Rock}
  n129 = OneMinus(None=MaterialFunctionCall(
    Transition Phase=Clamp(None=@n126)
    Height Texture=OneMinus(None=@n128.Height)
    Contrast=ScalarParameter() {defaultValue=3, parameterName=Desert_SandRock_Contrast}
   ) {desc=Base Color High end, materialFunction=&HeightLerp}.Lerp Alpha No Contrast)
  n130 = VertexNormalWS()
  n131 = Add(
   A=LinearInterpolate(
    A=ScalarParameter() {defaultValue=-1.5, group=Arid_SandRockBlend, parameterName=Arid_SandRock_LOW_1}
    B=ScalarParameter() {defaultValue=1.5, group=Arid_SandRockBlend, parameterName=Arid_SandRock_High_1}
    Alpha=ComponentMask(None=@n130) {b=true}
   )
   B=ScalarParameter() {defaultValue=-0.5429, group=Arid_SandRockBlend, parameterName=Arid_SandRock_Offset_1}
  )
  n132 = ComponentMask(None=MaterialFunctionCall(In=Clamp(None=@n131), Contrast=@n127) {materialFunction=&CheapContrast}.Result) {r=true}
  n133 = LinearInterpolate(
   B=LinearInterpolate(
    A=ScalarParameter() {defaultValue=1, parameterName=RockEdgeDesat}
    B=ScalarParameter() {defaultValue=1, parameterName=RockEdgeDesat_Arid}
    Alpha=@n121
   )
   Alpha=LinearInterpolate(A=Min(A=@n122, B=@n124), B=Min(A=@n129, B=@n132), Alpha=@n121)
  )
  n134 = Divide(A=@n34, B=@n37)
  n135 = LinearInterpolate(
   A=LinearInterpolate(A=Desaturation(None=@n120, Fraction=@n133) {desc=To Render Colormap}, B=Constant() {r=1}, Alpha=Clamp(None=@n134))
   B=@n65.Golf_Green_Color
   Alpha=MaterialFunctionCall(In=LandscapeLayerSample() {parameterName=Golf_Green}, Contrast=ScalarParameter() {parameterName=Golf_Mobile_LayerContrast}) {materialFunction=&CheapContrast}.Result
  )
  n136 = MaterialFunctionCall(
   Mask 2=TextureObject() {samplerType=SAMPLERTYPE_LinearColor, texture=&T_Athena_Terrain_Topo_Mask}
   UV Coordinates=LandscapeLayerCoords() {mappingScale=1525}
   GrassMask=@n115
   BaseTexture=@n135
  ) {materialFunction=&MF_TerrainTopoAdjustment}
  n137 = ScalarParameter() {parameterName=Arid_Edge_Desat_1}
  n138 = CollectionParameter() {collection=&MPC_Landscape, parameterName=Arid_SandSediment_Tint}
  n139 = LinearInterpolate(
   A=@n136.Result
   B=Multiply(A=Desaturation(None=@n136.Result, Fraction=@n137), B=ComponentMask(None=@n138) {b=true, g=true, r=true})
   Alpha=@n136.Result
  )
  n140 = LandscapeLayerSample() {parameterName=NoTopo}
  n141 = VertexNormalWS()
  n142 = Desaturation(
   None=Divide(
    A=LinearInterpolate(A=@n139, B=@n135, Alpha=@n140)
    B=CollectionParameter() {collection=&MPC_Landscape, parameterName=ColormapDivisor}
   )
   Fraction=SphereMask(
    A=CollectionParameter() {collection=&MPC_Landscape, parameterName=GrassEdgeZ}
    B=ComponentMask(None=@n141) {b=true}
    Radius=CollectionParameter() {collection=&MPC_Landscape, parameterName=GrassEdgeDesatWidth}
   ) {hardnessPercent=0}
  )
  n143 = CollectionParameter() {collection=&MPC_Landscape, parameterName=RenderColorMap}
  MaterialExpressionConstant_6 = Constant()
  MaterialExpressionConstant3Vector_15 = Constant3Vector() {desc=Mountain, constant=<rgba 0 1 0 0>}
  n144 = LandscapeLayerBlend!(
   Layer Grass_Only=@MaterialExpressionConstant_6
   Layer Grass_Topology=@MaterialExpressionConstant_6
   Layer Mountain_Grass=@MaterialExpressionConstant3Vector_15
  ) {layers [
    {
     blendType=LB_WeightBlend
     constHeightInput=0
     layerName=Grass_Only
     previewWeight=1
     constLayerInput=<0 0 0>
     heightInput {expression=null, inputName=None, mask=0, maskA=0, maskB=0, maskG=0, maskR=0, outputIndex=0}
     layerInput {
      expression=&self:MaterialExpressionConstant_6
      inputName=None
      mask=0
      maskA=0
      maskB=0
      maskG=0
      maskR=0
      outputIndex=0
     }
    }
    {
     blendType=LB_WeightBlend
     constHeightInput=0
     layerName=Grass_Topology
     previewWeight=1
     constLayerInput=<0 0 0>
     heightInput {expression=null, inputName=None, mask=0, maskA=0, maskB=0, maskG=0, maskR=0, outputIndex=0}
     layerInput {
      expression=&self:MaterialExpressionConstant_6
      inputName=None
      mask=0
      maskA=0
      maskB=0
      maskG=0
      maskR=0
      outputIndex=0
     }
    }
    {
     blendType=LB_WeightBlend
     constHeightInput=0
     layerName=Mountain_Grass
     previewWeight=1
     constLayerInput=<0 0 0>
     heightInput {expression=null, inputName=None, mask=0, maskA=0, maskB=0, maskG=0, maskR=0, outputIndex=0}
     layerInput {
      expression=&self:MaterialExpressionConstant3Vector_15
      inputName=None
      mask=0
      maskA=0
      maskB=0
      maskG=0
      maskR=0
      outputIndex=0
     }
    }
   ]}
  n145 = LinearInterpolate(
   A=LinearInterpolate(
    A=@n119
    B=MaterialFunctionCall!(UVW=@n142) {materialFunction=/Engine/ArtTools/RenderToTexture/MaterialFunctions/MarkRedOutside0-1Range}.Result
    Alpha=Saturate(None=@n143)
   )
   B=@n144
   Alpha=Saturate(None=CollectionParameter() {collection=&MPC_Landscape, parameterName=DisplayBiomeRGB})
  )
  n146 = StaticSwitchParameter(
   True=MaterialFunctionCall!(
    Mat_In=@n111.Blended Material
    BaseColor=StaticSwitchParameter(True=@n145, False=@n119) {group=Utilities, parameterName=EnableQuickColorRender}
   ) {materialFunction=/Game/Athena/Environments/Landscape/MaterialFunctions/MF_Athena_ReplaceBaseColor}.Result
   False=@n111.Blended Material
  ) {group=Rufus, parameterName=UseTopo}
  n147 = MaterialFunctionCall(
   Material=@n93.Blended Material
   NewBaseColor=MaterialFunctionCall(Material=@n84.Result) {materialFunction=&MatLayerBlend_BreakBaseColor}.BaseColor
   Mask=MaterialFunctionCall(In=LandscapeLayerSample() {parameterName=Forest, previewWeight=1}, Contrast=Constant() {r=2}) {materialFunction=&CheapContrast}.Result
  ) {materialFunction=&MatLayerBlend_OverrideBaseColor}
  n148 = MaterialFunctionCall(Material=@n98.Result) {materialFunction=&MatLayerBlend_BreakBaseColor}
  n149 = MaterialFunctionCall(Material=@n147.Blended Material, NewBaseColor=@n148.BaseColor, Mask=@n83.Alpha) {desc=medium, materialFunction=&MatLayerBlend_OverrideBaseColor}
  n150 = MaterialFunctionCall(Material=@n99.Result) {desc=Skip mud blend for Low/Mobile, materialFunction=&MatLayerBlend_BreakBaseColor}
  n151 = MaterialFunctionCall(Material=@n107) {materialFunction=&MatLayerBlend_BreakBaseColor}
  n152 = LandscapeLayerSample() {parameterName=Mud}
  n153 = MaterialFunctionCall(
   Material=MaterialFunctionCall(
    Material=MaterialFunctionCall(Material=@n147.Blended Material, NewBaseColor=@n148.BaseColor, Mask=@n152) {desc=Low, materialFunction=&MatLayerBlend_OverrideBaseColor}.Blended Material
    NewBaseColor=@n150.BaseColor
    Mask=@n101.Alpha
   ) {materialFunction=&MatLayerBlend_OverrideBaseColor}.Blended Material
   NewBaseColor=@n151.BaseColor
   Mask=LandscapeLayerSample() {parameterName=Gravel, previewWeight=1}
  ) {desc=Low, materialFunction=&MatLayerBlend_OverrideBaseColor}
  n154 = QualitySwitch(
   Default=@n146
   Low=@n153.Blended Material
   High=@n146
   Medium=MaterialFunctionCall(
    Material=MaterialFunctionCall(Material=@n149.Blended Material, NewBaseColor=@n150.BaseColor, Mask=@n101.Alpha) {desc=medium, materialFunction=&MatLayerBlend_OverrideBaseColor}.Blended Material
    NewBaseColor=@n151.BaseColor
    Mask=@n110.Alpha
   ) {desc=medium, materialFunction=&MatLayerBlend_OverrideBaseColor}.Blended Material
   Epic=@n146
  )
  n155 = BreakMaterialAttributes(Attr=FeatureLevelSwitch(Default=@n154, ES2=@n153.Blended Material, ES3_1=@n153.Blended Material, SM4_REMOVED=@n154, SM5=@n154, SM6=@n154))
  n156 = ComponentMask(None=@n155.BaseColor) {r=true}
  n157 = Multiply(A=@n156, B=ScalarParameter() {defaultValue=1, parameterName=SimpleMobileSpec})
  n158 = Clamp(None=GetMaterialAttributes(None=@n74.Material) {attributeGetTypes [#1]}.Specular)
  n159 = LandscapeLayerSample() {parameterName=Ice}
  n160 = LinearInterpolate(A=Constant() {r=0.1}, B=Constant() {r=0.25}, Alpha=@n159)
  n161 = LandscapeLayerSample() {parameterName=Snow}
  n162 = Add(A=@n161, B=@n159)
  n163 = StaticSwitchParameter(
   True=LinearInterpolate(
    A=FeatureLevelSwitch(Default=@n156, ES2=@n157, ES3_1=@n157, SM4_REMOVED=@n156, SM5=@n156, SM6=@n156)
    B=LinearInterpolate(A=@n158, B=@n160, Alpha=@n161)
    Alpha=Saturate(None=@n162)
   )
   False=@n155.Specular
  ) {defaultValue=true, parameterName=UseCheapSpec}
  n164 = MakeMaterialAttributes(BaseColor=@n155.BaseColor, Specular=@n163, Roughness=@n58, EmissiveColor=@n48, Normal=@n81, SubsurfaceColor=Constant())
  n165 = RuntimeVirtualTextureSampleParameter() {
   group=Rufus
   materialType=BaseColor_Normal_Specular_YCoCg
   parameterName=Landscape Read RVT
   virtualTexture=/BlastBerryMapContent/Game/Athena/Environments/Materials/RVT/Athena_RuntimeVirtualTexture
  }
  n166 = MakeMaterialAttributes(
   BaseColor=@n165.BaseColor
   Specular=@n165.BaseColor
   Roughness=@n165.BaseColor
   EmissiveColor=@n48
   Normal=Transform(None=@n165.Normal) {transformSourceType=TRANSFORMSOURCE_World, transformType=TRANSFORM_Tangent}
   SubsurfaceColor=Constant()
  )
  n167 = Saturate(None=Add(A=LandscapeLayerSample() {parameterName=Ice}, B=LandscapeLayerSample() {parameterName=Snow}))
  n168 = StaticSwitchParameter(
   True=@n164
   False=MaterialFunctionCall(NoRVT=@n164, RVT=BlendMaterialAttributes(A=@n166, B=@n164, Alpha=@n167)) {materialFunction=&FN_RVT_Switch_Material}.Result
  ) {group=Rufus, parameterName=RVT_Off}
  n169 = SetMaterialAttributes!(MaterialAttributes=@n168, Shading Model=ShadingModel(), Subsurface Color=Constant(), Opacity=Constant() {r=1}) {attributeSetTypes [#2, #3, #4]}
  n170 = SetMaterialAttributes!(
   MaterialAttributes=@n168
   Opacity=NamedRerouteUsage()
   Subsurface Color=NamedRerouteUsage()
   Shading Model=ShadingModel() {shadingModel=MSM_Subsurface}
  ) {attributeSetTypes [#4, #3, #2]}
  n171 = NamedRerouteUsage()
  MaterialFunctionCall!(In=FeatureLevelSwitch(Default=BlendMaterialAttributes(A=@n169, B=@n170, Alpha=@n171) {vertexAttributeBlendType=UseA}, ES2=@n168, ES3_1=@n168)) {materialFunction=/Game/Athena/Asteria/Environments/Landscape/Materials/MaterialFunctions/MF_Terrain_LumenModification}
  n172 = MaterialFunctionCall() {materialFunction=/Game/Athena/Environments/Landscape/MaterialFunctions/Standard/MF_Athena_RoadGravel}
  n173 = MaterialFunctionCall!(Gravel=AppendVector(A=@n172.Gravel, B=@n172.Gravel)) {materialFunction=/Game/Athena/Environments/Landscape/MaterialFunctions/Standard/MF_Athena_Rock_01}
  n174 = MaterialFunctionCall!(Rock=@n173.Result) {materialFunction=/BlastBerryMapContent/Game/Athena/Environments/Materials/MaterialFunctions/MF_Athena_IceRock_01}
  n175 = MaterialFunctionCall(UseLandscapeCoordsOnSwitch=StaticBool() {value=true}) {materialFunction=&MF_Athena_Snow_01}
  n176 = Abs(None=ComponentMask(None=VertexNormalWS()) {b=true})
  n177 = Constant() {r=0.975}
  n178 = MaterialFunctionCall(
   Transition Phase=Clamp(None=Multiply(A=@n176, B=@n177))
   Height Texture=Multiply(A=@n175.Height, B=Constant() {desc=Debug for multipler on Ice height blend, r=0.1})
   Contrast=Constant() {r=5}
  ) {materialFunction=&HeightLerp}
  n179 = MaterialFunctionCall(Base Material=@n174.Material, Top Material=@n175.Material, Alpha=@n178.Alpha) {materialFunction=&MatLayerBlend_Standard}
  NamedRerouteDeclaration(None=@n179.Blended Material) {name=ICE Mat, nodeColor=<rgba 1 0.8203 0 1>}
  n180 = VertexNormalWS()
  n181 = Clamp(None=Abs(None=ComponentMask(None=@n180) {b=true}))
  n182 = BreakMaterialAttributes(Attr=@n6.Result)
  n183 = Multiply(A=@n182.Opacity, B=Constant() {r=0.1})
  n184 = Constant() {r=5}
  NamedRerouteDeclaration(None=MaterialFunctionCall(
    Base Material=@n173.Result
    Top Material=@n175.Material
    Alpha=MaterialFunctionCall(Transition Phase=@n181, Height Texture=@n183, Contrast=@n184) {materialFunction=&HeightLerp}.Alpha
   ) {materialFunction=&MatLayerBlend_Standard}.Blended Material) {name=Snow Mat, nodeColor=<rgba 0 0.2969 1 1>}
  n185 = Multiply(A=Add(A=NamedRerouteUsage()) {constB=-0.5}) {constB=2}
  NamedRerouteDeclaration(None=Power(Base=Saturate(None=@n185)) {constExponent=1}) {name=SnowMaskModified, nodeColor=<rgba 1 0 0.4453 1>}
  NamedRerouteDeclaration(None=@n178.Alpha) {name=Cliff_Mask, nodeColor=<rgba 0 1 0.7891 1>}
  NamedRerouteDeclaration(None=LinearInterpolate(B=Constant() {r=0.95}, Alpha=NamedRerouteUsage()) {constA=1}) {name=snowOpacity, nodeColor=<rgba 1 0.5859 0 1>}
  n186 = VectorParameter() {group=PostRVTSnow, parameterName=SnowSSSColor, defaultValue=<rgba 0.0677 0.2393 0.5 1>}
  n187 = CameraVectorWS()
  n188 = PixelNormalWS()
  n189 = OneMinus(None=Saturate(None=DotProduct(A=@n187, B=@n188)))
  n190 = ScalarParameter() {defaultValue=4, group=PostRVTSnow, parameterName=SnowSpecularSelfShadowing}
  n191 = LinearInterpolate(
   A=ScalarParameter() {defaultValue=0.5, group=PostRVTSnow, parameterName=SnowSSSFacingExponent}
   Alpha=Saturate(None=Power(Base=@n189, Exp=@n190) {constExponent=3})
  )
  NamedRerouteDeclaration(None=Multiply(A=Multiply(A=@n186.RGB, B=@n191), B=NamedRerouteUsage())) {name=snowSubsurfaceColor, nodeColor=<rgba 0.0781 1 0 1>}
  NamedRerouteDeclaration(None=@n175.Height) {name=snow height, nodeColor=<rgba 0.6875 0 1 1>}
  NamedRerouteDeclaration(None=GetMaterialAttributes(None=@n179.Blended Material) {attributeGetTypes [#5]}.Base Color) {name=Ice Base Color, nodeColor=<rgba 0.8047 1 0 1>}
  NamedRerouteDeclaration(None=@n175.BaseColor) {name=Snow Base Color, nodeColor=<rgba 0 1 0.8594 1>}
  NamedRerouteDeclaration(None=Multiply(A=OneMinus(None=@n174.EmissiveMask), B=OneMinus(None=@n178.Alpha))) {name=Snow emmisive, nodeColor=<rgba 0 1 0.6719 1>}
  NamedRerouteDeclaration(None=@n178.Alpha) {name=Snow_normal_height, nodeColor=<rgba 1 0.5625 0 1>}
  NamedRerouteDeclaration(None=@n182.Opacity) {name=GRASS HEIGHT, nodeColor=<rgba 0 1 0.25 1>}
  n192 = LandscapeLayerSample() {parameterName=Ice}
  n193 = LandscapeLayerSample() {parameterName=Snow}
  NamedRerouteDeclaration(None=Saturate(None=Add(A=@n192, B=@n193))) {name=SnowFoliage, nodeColor=<rgba 1 0 0.7266 1>}
  n194 = Add(A=LandscapeLayerSample() {parameterName=Gravel}, B=LandscapeLayerSample() {parameterName=Forest})
  n195 = LandscapeLayerSample() {parameterName=Mud}
  n196 = Subtract(
   A=Max(A=LandscapeLayerSample() {parameterName=Snow}, B=LandscapeLayerSample() {parameterName=Ice})
   B=Add(A=Add(A=@n194, B=@n195), B=LandscapeLayerSample() {parameterName=Road})
  ) {desc=Mask out paths and dirt from emmisive}
  n197 = NamedRerouteUsage()
  NamedRerouteDeclaration(None=Clamp(None=Multiply(A=@n196, B=@n197))) {name=SnowMask, nodeColor=<rgba 1 0 0.2812 1>}
  NamedRerouteUsage()
  NamedRerouteUsage()
  NamedRerouteUsage()
  RuntimeVirtualTextureOutput!(
   BaseColor=@n155.BaseColor
   Specular=@n163
   Roughness=@n58
   Normal=Transform(None=@n81)
   WorldHeight=ComponentMask(None=WorldPosition().XYZ) {b=true}
   Opacity
   Mask
   Displacement
   Mask4
  )
 ]
}
